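# Animation Blueprint for the SurgeRaven female body cosmetic, following the attached parent mesh and adding secondary motion to clothing details.
# The copied parent pose feeds the body's Control Rig, then weighted multi-bone constraints place wrist and chest zipper bases and a button base from nearby body bones.
# AnimDynamics nodes then simulate the right front collar and three zippers with angular springs, tight limits and reduced gravity.
# The event graph contains no logic.
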AnimBlueprint /BRCosmetics/Characters/Player/Female/Medium/Bodies/F_MED_SurgeRaven_Pack/Meshes/F_MED_SurgeRaven_Pack_AnimBP
target_skeleton=&/F_MED_SurgeRaven_Pack_Skeleton
graphs [
 AnimGraph {nodes [
   n1 = Control Rig(Source=Copy Pose From Mesh() {bUseAttachedParent=true}) {controlRigAssetReference {blueprintRigClass=&/F_MED_SurgeRaven_Pack_CtrlRig.F_MED_SurgeRaven_Pack_CtrlRig_C}}
   n2 = "Constraint - base_zipper_l (PPPPPPPP)"!(
    ComponentPose=Local To Component(LocalPose=@n1)
    Alpha=1
    ConstraintWeights_0=0.7468
    ConstraintWeights_1:float=0.1316
    ConstraintWeights_2:float=0.0824
    ConstraintWeights_3:float=0.0182
    ConstraintWeights_4:float=0.0083
    ConstraintWeights_5:float=0.0079
    ConstraintWeights_6:float=0.004
    ConstraintWeights_7:float=0.0007
   ) -> (Pose) {
    boneToModify {boneName=base_zipper_l}
    constraintSetup [
     {targetBone {boneName=hand_l}}
     {targetBone {boneName=lowerarm_twist_01_l}}
     {targetBone {boneName=lowerarm_twist_02_l}}
     {targetBone {boneName=pinky_metacarpal_l}}
     {targetBone {boneName=lowerarm_l}}
     {targetBone {boneName=deform_wrist_up_l}}
     {targetBone {boneName=deform_wrist_dn_l}}
     {targetBone {boneName=ring_metacarpal_l}}
    ]
    constraintWeights [1,1,1,1,1,1,1,1]
   }
   n3 = "Constraint - base_zipper_chest (PPPPPPP)"!(
    ComponentPose=@n2
    Alpha=1
    ConstraintWeights_0=0.8577
    ConstraintWeights_1:float=0.0575
    ConstraintWeights_2:float=0.054
    ConstraintWeights_3:float=0.0286
    ConstraintWeights_4:float=0.0012
    ConstraintWeights_5:float=0.0006
    ConstraintWeights_6:float=0.0004
   ) -> (Pose) {
    boneToModify {boneName=base_zipper_chest}
    constraintSetup [
     {targetBone {boneName=spine_04}}
     {targetBone {boneName=spine_02}}
     {targetBone {boneName=spine_03}}
     {targetBone {boneName=spine_05}}
     {targetBone {boneName=deform_pec_l}}
     {targetBone {boneName=spine_01}}
     {targetBone {boneName=deform_pec_r}}
    ]
    constraintWeights [1,1,1,1,1,1,1]
   }
   n4 = "Constraint - base_zipper_r (PPPPPPPP)"!(
    ComponentPose=@n3
    Alpha=1
    ConstraintWeights_0=0.8526
    ConstraintWeights_1:float=0.0954
    ConstraintWeights_2:float=0.0247
    ConstraintWeights_3:float=0.0101
    ConstraintWeights_4:float=0.0087
    ConstraintWeights_5:float=0.0047
    ConstraintWeights_6:float=0.0022
    ConstraintWeights_7:float=0.0017
   ) -> (Pose) {
    boneToModify {boneName=base_zipper_r}
    constraintSetup [
     {targetBone {boneName=hand_r}}
     {targetBone {boneName=lowerarm_twist_01_r}}
     {targetBone {boneName=pinky_metacarpal_r}}
     {targetBone {boneName=deform_wrist_up_r}}
     {targetBone {boneName=lowerarm_twist_02_r}}
     {targetBone {boneName=lowerarm_r}}
     {targetBone {boneName=deform_wrist_dn_r}}
     {targetBone {boneName=ring_metacarpal_r}}
    ]
    constraintWeights [1,1,1,1,1,1,1,1]
   }
   n5 = "Constraint - base_button (PPPPPPPP)"!(
    ComponentPose=@n4
    Alpha=1
    ConstraintWeights_0=0.8573
    ConstraintWeights_1:float=0.1076
    ConstraintWeights_2:float=0.0181
    ConstraintWeights_3:float=0.0097
    ConstraintWeights_4:float=0.0037
    ConstraintWeights_5:float=0.0021
    ConstraintWeights_6:float=0.0007
    ConstraintWeights_7:float=0.0007
   ) -> (Pose) {
    boneToModify {boneName=base_button}
    constraintSetup [
     {targetBone {boneName=pelvis}}
     {targetBone {boneName=spine_02}}
     {targetBone {boneName=deform_groin_l}}
     {targetBone {boneName=spine_04}}
     {targetBone {boneName=ik_hand_gun}}
     {targetBone {boneName=spine_03}}
     {targetBone {boneName=thigh_l}}
     {targetBone {boneName=deform_groin_base_l}}
    ]
    constraintWeights [1,1,1,1,1,1,1,1]
   }
   n6 = Anim Dynamics - Bone: dyn_collar_front_r!(ComponentPose=@n5, bAlphaBoolEnabled:bool=True, GravityOverride:Vector="0,0,-980", ExternalForce="0,0,0") -> (Pose) {
    alphaInputType=Bool
    angularDampingOverride=0.9999
    angularSpringConstant=2400
    bAngularSpring=true
    bOverrideAngularDamping=true
    bUseGravityOverride=true
    gravityScale=0.35
    simulationSpace=RootRelative
    alphaBoolBlend {blendInTime=2, blendOption=Cubic}
    boundBone {boneName=dyn_collar_front_r}
    componentLinearAccScale=<1 1 1>
    componentLinearVelScale=<1 1 1>
    gravityOverride {z=-980}
    physicsBodyDefinitions [{
      boundBone {boneName=dyn_collar_front_r}
      boxExtents=<25 25 25>
      constraintSetup {angularLimitsMax=<10 10 10>, angularLimitsMin=<-10 0 -10>, angularTarget=<1 0 0>}
      localJointOffset=<3 0 0>
     }]
    relativeSpaceBone {boneName=head}
   }
   n7 = Anim Dynamics - Bone: dyn_zipper_l!(ComponentPose=@n6, bAlphaBoolEnabled:bool=True, GravityOverride:Vector="0,0,-980", ExternalForce="0,0,0") -> (Pose) {
    alphaInputType=Bool
    angularDampingOverride=0.9999
    angularSpringConstant=4000
    bAngularSpring=true
    bOverrideAngularDamping=true
    bUseGravityOverride=true
    gravityScale=0.35
    simulationSpace=BoneRelative
    alphaBoolBlend {blendInTime=2, blendOption=Cubic}
    boundBone {boneName=dyn_zipper_l}
    componentLinearAccScale=<1 1 1>
    componentLinearVelScale=<1 1 1>
    gravityOverride {z=-980}
    physicsBodyDefinitions [{
      boundBone {boneName=dyn_zipper_l}
      boxExtents=<25 25 25>
      constraintSetup {angularLimitsMax=<0 0 0>, angularLimitsMin=<0 0 -20>, angularTarget=<1 0 0>}
      localJointOffset=<3 0 0>
     }]
    relativeSpaceBone {boneName=lowerarm_twist_02_l}
   }
   n8 = Anim Dynamics - Bone: dyn_zipper_r!(ComponentPose=@n7, bAlphaBoolEnabled:bool=True, GravityOverride:Vector="0,0,-980", ExternalForce="0,0,0") -> (Pose) {
    alphaInputType=Bool
    angularDampingOverride=0.9999
    angularSpringConstant=2000
    bAngularSpring=true
    bOverrideAngularDamping=true
    bUseGravityOverride=true
    gravityScale=0.35
    simulationSpace=BoneRelative
    alphaBoolBlend {blendInTime=2, blendOption=Cubic}
    boundBone {boneName=dyn_zipper_r}
    componentLinearAccScale=<1 1 1>
    componentLinearVelScale=<1 1 1>
    gravityOverride {z=-980}
    physicsBodyDefinitions [{
      boundBone {boneName=dyn_zipper_r}
      boxExtents=<25 25 25>
      constraintSetup {angularLimitsMax=<0 0 0>, angularLimitsMin=<0 0 -90>, angularTarget=<1 0 0>}
      localJointOffset=<2 0 0>
     }]
    relativeSpaceBone {boneName=lowerarm_twist_02_r}
   }
   n9 = Anim Dynamics - Bone: dyn_zipper_chest!(ComponentPose=@n8, bAlphaBoolEnabled:bool=True, GravityOverride:Vector="0,0,-980", ExternalForce="0,0,0") -> (Pose) {
    alphaInputType=Bool
    angularDampingOverride=0.9999
    angularSpringConstant=5000
    bAngularSpring=true
    bOverrideAngularDamping=true
    bUseGravityOverride=true
    gravityScale=0.35
    simulationSpace=RootRelative
    alphaBoolBlend {blendInTime=2, blendOption=Cubic}
    boundBone {boneName=dyn_zipper_chest}
    componentLinearAccScale=<1 1 1>
    componentLinearVelScale=<1 1 1>
    gravityOverride {z=-980}
    physicsBodyDefinitions [{
      boundBone {boneName=dyn_zipper_chest}
      boxExtents=<25 25 25>
      constraintSetup {angularLimitsMax=<0 0 0>, angularLimitsMin=<0 -40 0>, angularTarget=<1 0 0>}
      localJointOffset=<1 0 0>
     }]
    relativeSpaceBone {boneName=head}
   }
   Output Pose(Result=Component To Local(ComponentPose=@n9)) {layerGroup=None}
  ]}
 EventGraph {nodes [Event BlueprintUpdateAnimation()]}
]
variables []
