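# Physics asset for the front hair (FE2) face-accessory part of the SurgeRaven female medium character, using only capsule bodies.
# Head, neck, upper spine, clavicle, upper-arm and collar capsules approximate the body around the hair.
# Two two-link dyn hair chains, left upper and right front-to-mid, hang from base hair bodies with narrow swing and twist limits.
# Base hair bodies link to the head, and head to neck_02, with free angular motion.
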PhysicsAsset /BRCosmetics/Characters/Player/Female/Medium/Bodies/F_MED_SurgeRaven_Pack/Meshes/Parts/F_MED_SurgeRaven_Pack_FaceAcc_FrontHair_FE2_Physics
bodies {
 base_collar_front_l [{shapeType=Capsule, length=4, radius=2, shapeName=base_collar_front_l_capsule, center=<0 1.6818 0>, rotation=<rot 0 -90 0>}]
 base_collar_front_r [{shapeType=Capsule, length=6, radius=2.5, shapeName=base_collar_front_r_capsule, center=<0 2.2952 -0.6692>, rotation=<rot 30.0001 -90 0.0004>}]
 base_fr_hair_l [{shapeType=Capsule, length=0.505, radius=0.505, shapeName=base_fr_hair_l_capsule, rotation=<rot 0 -90 0>}]
 base_fr_hair_l_02 [{shapeType=Capsule, length=0.505, radius=0.505, shapeName=base_fr_hair_l_02_capsule, rotation=<rot 0 -90 0>}]
 base_fr_hair_r [{shapeType=Capsule, length=0.505, radius=0.505, shapeName=base_fr_hair_r_capsule, rotation=<rot 0 -90 0>}]
 base_hairtop [{shapeType=Capsule, length=1, radius=1, shapeName=base_hairtop_capsule, rotation=<rot 90 0 0>}]
 clavicle_l [{shapeType=Capsule, length=5.5, radius=8, shapeName=clavicle_l_capsule, center=<7.795 -0.9969 0.4189>, rotation=<rot 90 0.0004 0>}]
 clavicle_r [{shapeType=Capsule, length=4, radius=6, shapeName=clavicle_l_capsule, center=<-6.3914 1.3685 1.5437>, rotation=<rot 80 -0.0007 -0.0007>}]
 dyn_fr_hair_r [{shapeType=Capsule, length=6.5, radius=1.5, shapeName=dyn_fr_hair_r_capsule, center=<3.9831 0 0>, rotation=<rot -90 179.156 -179.1564>}]
 dyn_fr_midhair_r [{
   shapeType=Capsule
   length=4.5
   radius=3
   shapeName=dyn_fr_upperhair_l_02_capsule
   center=<4.3382 0 1.4412>
   rotation=<rot 50.0001 179.9999 180>
  }]
 dyn_fr_upperhair_l_01 [{shapeType=Capsule, length=2.5, radius=2, shapeName=dyn_fr_upperhair_l_01_capsule, center=<6 0 0>, rotation=<rot 90 0 0>}]
 dyn_fr_upperhair_l_02 [{shapeType=Capsule, length=6, radius=2.5, shapeName=dyn_fr_upperhair_l_02_capsule, center=<6 0 -2.6>, rotation=<rot 90 0 0>}]
 head [{shapeType=Capsule, length=6.7343, radius=5.75, shapeName=head_capsule, center=<1.6155 1.5676 0>, rotation=<rot 90 0 0>}]
 mover_clavicle_r [{
   shapeType=Capsule
   length=3.8998
   radius=6.8005
   shapeName=clavicle_l_capsule
   center=<-5.2125 0.9604 2.4239>
   rotation=<rot 88.5323 -179.9983 -179.9983>
  }]
 neck_01 [{shapeType=Capsule, length=5.5, radius=4.5, shapeName=neck_01_capsule, rotation=<rot 90 0 0>}]
 neck_02 [{shapeType=Capsule, length=28, radius=4.5, shapeName=neck_01_capsule, center=<0 -5.038 0>, rotation=<rot 0 179.9999 -180>}]
 spine_05 [{shapeType=Capsule, length=11.5, radius=10, shapeName=spine_05_capsule, center=<-1.096 -1.1712 0>}]
 upperarm_l [{shapeType=Capsule, length=15.5, radius=5.5, shapeName=upperarm_l_capsule, center=<12.4723 0 0>, rotation=<rot 90 0 0>}]
 upperarm_r [{shapeType=Capsule, length=16.5, radius=4, shapeName=upperarm_l_capsule, center=<-12.4723 0.7632 -3.555>, rotation=<rot 90 0 0>}]
}
constraints [
 {
  bone1Name=base_hairtop
  bone2Name=head
  swing1LimitDegrees=45
  swing1Motion=Free
  swing2LimitDegrees=45
  swing2Motion=Free
  twistLimitDegrees=45
  twistMotion=Free
 }
 {
  bone1Name=head
  bone2Name=neck_02
  swing1LimitDegrees=45
  swing1Motion=Free
  swing2LimitDegrees=45
  swing2Motion=Free
  twistLimitDegrees=45
  twistMotion=Free
 }
 {
  bone1Name=base_fr_hair_l
  bone2Name=head
  swing1LimitDegrees=45
  swing1Motion=Free
  swing2LimitDegrees=45
  swing2Motion=Free
  twistLimitDegrees=45
  twistMotion=Free
 }
 {
  bone1Name=base_fr_hair_l_02
  bone2Name=head
  swing1LimitDegrees=45
  swing1Motion=Free
  swing2LimitDegrees=45
  swing2Motion=Free
  twistLimitDegrees=45
  twistMotion=Free
 }
 {bone1Name=dyn_fr_upperhair_l_01, bone2Name=base_fr_hair_l_02, swing1LimitDegrees=15, swing2LimitDegrees=15, twistLimitDegrees=5}
 {bone1Name=dyn_fr_upperhair_l_02, bone2Name=dyn_fr_upperhair_l_01, swing1LimitDegrees=15, swing2LimitDegrees=35, twistLimitDegrees=5}
 {
  bone1Name=base_fr_hair_r
  bone2Name=head
  swing1LimitDegrees=45
  swing1Motion=Free
  swing2LimitDegrees=45
  swing2Motion=Free
  twistLimitDegrees=45
  twistMotion=Free
 }
 {bone1Name=dyn_fr_hair_r, bone2Name=base_fr_hair_r, swing1LimitDegrees=15, swing2LimitDegrees=25, twistLimitDegrees=5}
 {bone1Name=dyn_fr_midhair_r, bone2Name=dyn_fr_hair_r, swing1LimitDegrees=25, swing2LimitDegrees=35, twistLimitDegrees=10}
]
