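# Physics asset for the SurgeRaven female medium body's FaceAcc part, defining capsule bodies for hair physics.
# Head, neck, upper spine, clavicle and upper-arm capsules cover the body, alongside small base bodies for hair and collar.
# Left, middle and right back-hair chains and two front chains hang off base bones with tight swing and twist limits.
# Extra angularly free constraints cross-link neighboring strands, tying front hair to back and side back strands to middle.
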PhysicsAsset /BRCosmetics/Characters/Player/Female/Medium/Bodies/F_MED_SurgeRaven_Pack/Meshes/Parts/F_MED_SurgeRaven_Pack_FaceAcc_Physics
bodies {
 base_backhair_l [{shapeType=Capsule, length=0.51, radius=0.5, shapeName=base_backhair_l_capsule, center=<1.5009 0 0>, rotation=<rot 90 0 0>}]
 base_backhair_m [{shapeType=Capsule, length=0.51, radius=0.5, shapeName=base_backhair_m_capsule, center=<1.5009 0 0>, rotation=<rot 90 0 0>}]
 base_backhair_r [{shapeType=Capsule, length=0.505, radius=0.505, shapeName=base_backhair_r_capsule, center=<1.5009 0 0>, rotation=<rot 90 0 0>}]
 base_collar_front_l [{shapeType=Capsule, length=0.505, radius=0.505, shapeName=base_collar_front_l_capsule, rotation=<rot 0 -90 0>}]
 base_collar_front_r [{shapeType=Capsule, length=0.505, radius=0.505, shapeName=base_collar_front_r_capsule, rotation=<rot 0 -90 0>}]
 base_fr_02_hair_r [{shapeType=Capsule, length=1, radius=1, shapeName=base_fr_02_hair_r_capsule}]
 base_fr_hair_l [{shapeType=Capsule, length=0.505, radius=0.505, shapeName=base_fr_hair_l_capsule, rotation=<rot 0 -90 0>}]
 base_hair [{shapeType=Capsule, length=0.505, radius=0.505, shapeName=base_hair_capsule, rotation=<rot 90 0 0>}]
 base_hairtop [{shapeType=Capsule, length=1, radius=1, shapeName=base_hairtop_capsule, rotation=<rot 90 0 0>}]
 clavicle_l [{shapeType=Capsule, length=5.5, radius=5, shapeName=clavicle_l_capsule, center=<6.6636 -2.0167 0>, rotation=<rot 80 -0.0017 -0.0017>}]
 clavicle_r [{shapeType=Capsule, length=7, radius=5.5, shapeName=clavicle_l_capsule, center=<-6.6013 2.0123 0.3537>, rotation=<rot 80 -0.0007 -0.0007>}]
 dyn_backhair_l_01 [{shapeType=Capsule, length=2.0257, radius=0.7781, shapeName=dyn_backhair_l_01_capsule, center=<1.5009 0 0.5473>, rotation=<rot 90 0 0>}]
 dyn_backhair_l_02 [{shapeType=Capsule, length=1.5, radius=3, shapeName=dyn_backhair_l_02_capsule, center=<2.1234 0 0.5473>, rotation=<rot 90 0 0>}]
 dyn_backhair_l_03 [{shapeType=Capsule, length=1, radius=2.5, shapeName=dyn_backhair_r_03_capsule, center=<2.8254 0 0>, rotation=<rot 90 0 0>}]
 dyn_backhair_l_04 [{
   shapeType=Capsule
   length=4.3855
   radius=2.4145
   shapeName=dyn_backhair_l_04_capsule
   center=<6.2155 -0.1483 1.1978>
   rotation=<rot 73.548 -142.8551 -64.1524>
  }]
 dyn_backhair_l_05 [{shapeType=Capsule, length=4, radius=2, shapeName=dyn_backhair_l_05_capsule, center=<2.2757 0 0>, rotation=<rot 59.9999 -89.9994 -179.9999>}]
 dyn_backhair_m_01 [{shapeType=Capsule, length=4.0283, radius=1.9674, shapeName=dyn_backhair_m_01_capsule, center=<2.0868 0 0>, rotation=<rot 0 0 90>}]
 dyn_backhair_m_02 [{shapeType=Capsule, length=2.5, radius=2, shapeName=dyn_backhair_m_02_capsule, center=<2.2881 0 0>, rotation=<rot 0 0 90>}]
 dyn_backhair_m_03 [{shapeType=Capsule, length=6, radius=2, shapeName=dyn_backhair_m_03_capsule, center=<2.1402 0 0>, rotation=<rot 0 0 90>}]
 dyn_backhair_m_04 [{shapeType=Capsule, length=1, radius=3, shapeName=dyn_backhair_r_03_capsule, center=<2.8254 0 2>, rotation=<rot 90 0 0>}]
 dyn_backhair_m_05 [{shapeType=Capsule, length=3.5078, radius=3.8539, shapeName=dyn_backhair_r_03_capsule, center=<5.551 0 1.1746>, rotation=<rot 0 0 90>}]
 dyn_backhair_r_01 [{shapeType=Capsule, length=0.505, radius=0.505, shapeName=dyn_backhair_r_01_capsule, center=<1.5009 0 0.5473>, rotation=<rot 90 0 0>}]
 dyn_backhair_r_02 [{shapeType=Capsule, length=2, radius=2.5, shapeName=dyn_backhair_r_02_capsule, center=<1.5009 0 0.5473>, rotation=<rot 90 0 0>}]
 dyn_backhair_r_03 [{shapeType=Capsule, length=1, radius=2.5, shapeName=dyn_backhair_r_03_capsule, center=<2.8254 0 0>, rotation=<rot 90 0 0>}]
 dyn_backhair_r_04 [{
   shapeType=Capsule
   length=2.6479
   radius=3.3521
   shapeName=dyn_backhair_r_04_capsule
   center=<6.2497 -2.0937 -0.0194>
   rotation=<rot 90 40 0>
  }]
 dyn_fr_02_hair_r [{shapeType=Capsule, length=4.5, radius=2.5, shapeName=dyn_fr_02_hair_r_capsule, center=<4.635 0 0>, rotation=<rot 90 0 0>}]
 dyn_fr_hair_l [{
   shapeType=Capsule
   length=4.2578
   radius=2.5
   shapeName=dyn_fr_hair_l_capsule
   center=<4.851 -0.6604 0>
   rotation=<rot 90 -179.4032 -174.4183>
  }]
 dyn_fr_lowerhair_l_01 [{shapeType=Capsule, length=2, radius=2.5, shapeName=dyn_fr_lowerhair_l_01_capsule, center=<3.0426 0 0>, rotation=<rot 90 0 0>}]
 dyn_fr_lowerhair_l_02 [{shapeType=Capsule, length=4, radius=2, shapeName=dyn_fr_lowerhair_l_02_capsule, center=<3 0 6>, rotation=<rot 0 90 90>}]
 dyn_fr_lowerhair_r_01 [{shapeType=Capsule, length=4, radius=3, shapeName=dyn_fr_lowerhair_l_01_capsule, center=<3.2457 0.4027 1>, rotation=<rot 90 -0.0004 0>}]
 dyn_fr_lowerhair_r_02 [{
   shapeType=Capsule
   length=5.5607
   radius=3
   shapeName=dyn_fr_lowerhair_r_02_capsule
   center=<3.3171 -1.2372 2.1156>
   rotation=<rot 65.0005 -124.9993 150.0008>
  }]
 head [{shapeType=Capsule, length=6.7343, radius=5.5674, shapeName=head_capsule, center=<1.6155 1.5676 0>, rotation=<rot 90 0 0>}]
 mover_clavicle_r [{shapeType=Capsule, length=4.5, radius=6, shapeName=clavicle_l_capsule, center=<-6.6636 2.3401 0>, rotation=<rot 80 -0.0007 -0.0007>}]
 mover_upperarm_r [{
   shapeType=Capsule
   length=10.6329
   radius=4.9532
   shapeName=upperarm_l_capsule
   center=<-5.4593 0.2245 0.7534>
   rotation=<rot 85.0289 -2.5806 -0.0016>
  }]
 neck_01 [{shapeType=Capsule, length=5.5, radius=4.5, shapeName=neck_01_capsule, rotation=<rot 90 0 0>}]
 spine_05 [{shapeType=Capsule, length=12, radius=9.5, shapeName=spine_05_capsule, center=<1.1757 -0.4463 0>}]
 upperarm_l [{shapeType=Capsule, length=15.5, radius=5.5, shapeName=upperarm_l_capsule, center=<12.4723 0 0>, rotation=<rot 90 0 0>}]
 upperarm_r [{shapeType=Capsule, length=15.5, radius=5.5, shapeName=upperarm_l_capsule, center=<-12.4723 0 0>, rotation=<rot 90 0 0>}]
}
constraints [
 {
  bone1Name=base_collar_front_l
  bone2Name=spine_05
  swing1LimitDegrees=45
  swing1Motion=Free
  swing2LimitDegrees=45
  swing2Motion=Free
  twistLimitDegrees=45
  twistMotion=Free
 }
 {
  bone1Name=base_collar_front_r
  bone2Name=spine_05
  swing1LimitDegrees=45
  swing1Motion=Free
  swing2LimitDegrees=45
  swing2Motion=Free
  twistLimitDegrees=45
  twistMotion=Free
 }
 {
  bone1Name=head
  bone2Name=neck_01
  swing1LimitDegrees=45
  swing1Motion=Free
  swing2LimitDegrees=45
  swing2Motion=Free
  twistLimitDegrees=45
  twistMotion=Free
 }
 {
  bone1Name=base_hair
  bone2Name=head
  swing1LimitDegrees=45
  swing1Motion=Free
  swing2LimitDegrees=45
  swing2Motion=Free
  twistLimitDegrees=45
  twistMotion=Free
 }
 {
  bone1Name=base_backhair_l
  bone2Name=base_hair
  swing1LimitDegrees=45
  swing1Motion=Free
  swing2LimitDegrees=45
  swing2Motion=Free
  twistLimitDegrees=45
  twistMotion=Free
 }
 {bone1Name=dyn_backhair_l_01, bone2Name=base_backhair_l, swing1LimitDegrees=8, swing2LimitDegrees=20, twistLimitDegrees=5}
 {bone1Name=dyn_backhair_l_02, bone2Name=dyn_backhair_l_01, swing1LimitDegrees=8, swing2LimitDegrees=15, twistLimitDegrees=5}
 {bone1Name=dyn_backhair_l_03, bone2Name=dyn_backhair_l_02, swing1LimitDegrees=8, swing2LimitDegrees=25, twistLimitDegrees=5}
 {bone1Name=dyn_backhair_l_04, bone2Name=dyn_backhair_l_03, swing1LimitDegrees=8, swing2LimitDegrees=25, twistLimitDegrees=5}
 {bone1Name=dyn_backhair_l_05, bone2Name=dyn_backhair_l_03, swing1LimitDegrees=5, swing2LimitDegrees=25, twistLimitDegrees=5}
 {
  bone1Name=base_backhair_m
  bone2Name=base_hair
  swing1LimitDegrees=45
  swing1Motion=Free
  swing2LimitDegrees=45
  swing2Motion=Free
  twistLimitDegrees=45
  twistMotion=Free
 }
 {bone1Name=dyn_backhair_m_01, bone2Name=base_backhair_m, swing1LimitDegrees=8, swing2LimitDegrees=20, twistLimitDegrees=5}
 {bone1Name=dyn_backhair_m_02, bone2Name=dyn_backhair_m_01, swing1LimitDegrees=8, swing2LimitDegrees=25, twistLimitDegrees=5}
 {bone1Name=dyn_backhair_m_03, bone2Name=dyn_backhair_m_02, swing1LimitDegrees=8, swing2LimitDegrees=25, twistLimitDegrees=5}
 {bone1Name=dyn_backhair_m_04, bone2Name=dyn_backhair_m_03, swing1LimitDegrees=8, swing2LimitDegrees=25, twistLimitDegrees=5}
 {bone1Name=dyn_backhair_m_05, bone2Name=dyn_backhair_m_04, swing1LimitDegrees=10, swing2LimitDegrees=25, twistLimitDegrees=5}
 {
  bone1Name=base_backhair_r
  bone2Name=base_hair
  swing1LimitDegrees=45
  swing1Motion=Free
  swing2LimitDegrees=45
  swing2Motion=Free
  twistLimitDegrees=45
  twistMotion=Free
 }
 {bone1Name=dyn_backhair_r_01, bone2Name=base_backhair_r, swing1LimitDegrees=8, swing2LimitDegrees=20, twistLimitDegrees=5}
 {bone1Name=dyn_backhair_r_02, bone2Name=dyn_backhair_r_01, swing1LimitDegrees=8, swing2LimitDegrees=25, twistLimitDegrees=5}
 {bone1Name=dyn_backhair_r_03, bone2Name=dyn_backhair_r_02, swing1LimitDegrees=8, swing2LimitDegrees=25, twistLimitDegrees=5}
 {bone1Name=dyn_backhair_r_04, bone2Name=dyn_backhair_r_03, swing1LimitDegrees=8, swing2LimitDegrees=15, twistLimitDegrees=5}
 {
  bone1Name=base_fr_02_hair_r
  bone2Name=base_hair
  swing1LimitDegrees=45
  swing1Motion=Free
  swing2LimitDegrees=45
  swing2Motion=Free
  twistLimitDegrees=45
  twistMotion=Free
 }
 {bone1Name=dyn_fr_02_hair_r, bone2Name=base_fr_02_hair_r, swing1LimitDegrees=5, swing2LimitDegrees=20, twistLimitDegrees=5}
 {bone1Name=dyn_fr_lowerhair_r_01, bone2Name=dyn_fr_02_hair_r, swing1LimitDegrees=10, swing2LimitDegrees=20, twistLimitDegrees=5}
 {bone1Name=dyn_fr_lowerhair_r_02, bone2Name=dyn_fr_lowerhair_r_01, swing1LimitDegrees=15, swing2LimitDegrees=40, twistLimitDegrees=5}
 {
  bone1Name=base_fr_hair_l
  bone2Name=base_hair
  swing1LimitDegrees=45
  swing1Motion=Free
  swing2LimitDegrees=45
  swing2Motion=Free
  twistLimitDegrees=45
  twistMotion=Free
 }
 {bone1Name=dyn_fr_hair_l, bone2Name=base_fr_hair_l, swing1LimitDegrees=15, swing2LimitDegrees=25, twistLimitDegrees=5}
 {bone1Name=dyn_fr_lowerhair_l_01, bone2Name=dyn_fr_hair_l, swing1LimitDegrees=15, swing2LimitDegrees=35, twistLimitDegrees=5}
 {bone1Name=dyn_fr_lowerhair_l_02, bone2Name=dyn_fr_lowerhair_l_01, swing1LimitDegrees=15, swing2LimitDegrees=25, twistLimitDegrees=5}
 {
  bone1Name=base_hairtop
  bone2Name=base_hair
  swing1LimitDegrees=45
  swing1Motion=Free
  swing2LimitDegrees=45
  swing2Motion=Free
  twistLimitDegrees=45
  twistMotion=Free
 }
 {
  bone1Name=dyn_fr_lowerhair_r_01
  bone2Name=dyn_backhair_r_03
  swing1LimitDegrees=45
  swing1Motion=Free
  swing2LimitDegrees=45
  swing2Motion=Free
  twistLimitDegrees=45
  twistMotion=Free
 }
 {
  bone1Name=dyn_fr_lowerhair_l_02
  bone2Name=dyn_backhair_l_05
  swing1LimitDegrees=45
  swing1Motion=Free
  swing2LimitDegrees=45
  swing2Motion=Free
  twistLimitDegrees=45
  twistMotion=Free
 }
 {
  bone1Name=dyn_backhair_r_04
  bone2Name=dyn_backhair_m_04
  swing1LimitDegrees=45
  swing1Motion=Free
  swing2LimitDegrees=45
  swing2Motion=Free
  twistLimitDegrees=45
  twistMotion=Free
 }
 {
  bone1Name=dyn_backhair_l_04
  bone2Name=dyn_backhair_m_04
  swing1LimitDegrees=45
  swing1Motion=Free
  swing2LimitDegrees=45
  swing2Motion=Free
  twistLimitDegrees=45
  twistMotion=Free
 }
 {
  bone1Name=dyn_fr_lowerhair_l_01
  bone2Name=dyn_backhair_l_02
  swing1LimitDegrees=45
  swing1Motion=Free
  swing2LimitDegrees=45
  swing2Motion=Free
  twistLimitDegrees=45
  twistMotion=Free
 }
]
