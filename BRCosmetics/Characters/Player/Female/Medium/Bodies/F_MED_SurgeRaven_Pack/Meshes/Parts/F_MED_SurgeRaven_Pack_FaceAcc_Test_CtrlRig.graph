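# Test Control Rig for the SurgeRaven female character's face accessory part, adjusting hair base bones to correct their pose.
# A constant transform offset runs first, including on the front right hair.
# A spherical pose reader on the head then weights back-hair rotation and translation offsets.
# The Is Skydiving bool switches on a front-left hair rotation; another front-left offset node is left unconnected.
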ControlRigBlueprint /BRCosmetics/Characters/Player/Female/Medium/Bodies/F_MED_SurgeRaven_Pack/Meshes/Parts/F_MED_SurgeRaven_Pack_FaceAcc_Test_CtrlRig
graphs [
 RigVMFunctionLibrary {links [], nodes []}
 RigVMModel {nodes [
   RigUnit_BeginExecution = &RigUnit_BeginExecution()
   &RigUnit_ModifyTransforms(
    ItemToModify="((Item=(Type=Bone,Name=\"None\"),Transform=(Rotation=(X=0,Y=0,Z=0,W=1),Translation=(X=0,Y=0,Z=0),Scale3D=(X=1,Y=1,Z=1))),(Item=(Type=Bone,Name=\"base_fr_hair_r\"),Transform=(Rotation=(X=0,Y=0,Z=0,W=1),Translation=(X=-1,Y=0,Z=0),Scale3D=(X=1,Y=1,Z=1))),(Item=(T ...(truncated)"
    Weight=1
    WeightMinimum=0
    ExecutePin=@RigUnit_BeginExecution.ExecutePin
   )
   > n1 = &RigUnit_SphericalPoseReader(
    ActiveRegionScaleFactors=(PositiveWidth=1,NegativeWidth=0.2,PositiveHeight=1.5,NegativeHeight=1.5)
    ActiveRegionSize=0.1
    Debug=(bDrawDebug=false,bDraw2D=false,bDrawLocalAxes=false,DebugScale=25,DebugSegments=20,DebugThickness=0.25)
    DriverAxis=(X=0,Y=1,Z=0)
    DriverItem=(Type=Bone,Name="head")
    FalloffRegionScaleFactors=(PositiveWidth=3,NegativeWidth=1.8,PositiveHeight=2.5,NegativeHeight=2.5)
    FalloffSize=0.2
    FlipHeightScaling=False
    FlipWidthScaling=False
    OptionalParentItem=(Type=Bone,Name="None")
    RotationOffset=(X=90,Y=0,Z=-55)
   )
   > &RigUnit_ModifyTransforms(
    ItemToModify="((Item=(Type=Bone,Name=\"base_backhair_m\"),Transform=(Rotation=(X=0,Y=-0.3827,Z=0,W=0.9239),Translation=(X=0,Y=0,Z=5),Scale3D=(X=1,Y=1,Z=1))),(Item=(Type=Bone,Name=\"base_backhair_l\"),Transform=(Rotation=(X=0,Y=-0.3827,Z=0,W=0.9239),Translation=(X=0,Y=0,Z=5) ...(truncated)"
    WeightMinimum=0
    Weight=@n1.OutputParam
   )
   > &RigUnit_ModifyTransforms(
    ItemToModify=((Item=(Type=Bone,Name="base_fr_hair_l"),Transform=(Rotation=(X=0,Y=-0.3827,Z=0,W=0.9239),Translation=(X=0,Y=0,Z=0),Scale3D=(X=1,Y=1,Z=1))))
    WeightMinimum=0
    Weight=RigVMDispatchNode(False=0, True=1, Condition=RigVMVariableNode() {variable=Is Skydiving}.Value) {template="DISPATCH_RigVMDispatch_If(in Condition,in True,in False,out Result)"}.Result
   )
   &RigUnit_ModifyTransforms(
    ExecutePin=()
    ItemToModify=((Item=(Type=Bone,Name="base_fr_hair_l"),Transform=(Rotation=(X=0,Y=0,Z=0,W=1),Translation=(X=-3,Y=0,Z=0),Scale3D=(X=1,Y=1,Z=1))))
    Weight=1
    WeightMinimum=0
   )
   &RigUnit_PrepareForExecution()
  ]}
]
hierarchy [
 root {loc=<0 0 0>, rot=<0 0 0 1>}
 pelvis {parent=root, loc=<0 -0.5301 98.0571>, rot=<0.0127 0.707 -0.0127 -0.707>}
 spine_01 {parent=pelvis, loc=<6.1208 -0.9328 0>, rot=<0 0 0.034 -0.9994>}
 spine_02 {parent=spine_01, loc=<6.5997 -0.2095 0>, rot=<0 0 0.0238 -0.9997>}
 spine_03 {parent=spine_02, loc=<6.4914 -0.1237 0>, rot=<0 0 0.0149 -0.9999>}
 spine_04 {parent=spine_03, loc=<6.9598 0.0794 0>, rot=<0 0 -0.0132 -0.9999>}
 spine_05 {parent=spine_04, loc=<6.9987 0 0>, rot=<0 0 -0.0015 -1>}
 neck_01 {parent=spine_05, loc=<11.3138 0.1529 0>, rot=<0 0 -0.2003 -0.9797>}
 neck_02 {parent=neck_01, loc=<4.5627 0 0>, rot=<0 0 0.0504 -0.9987>}
 head {parent=neck_02, loc=<5.3914 0 0>, rot=<0 0 0.115 -0.9934>}
 base_hair {parent=head, loc=<0 0 0>, rot=<0.7071 -0.0009 0.0009 -0.7071>}
 base_hairtop {parent=base_hair, loc=<12.0279 0 0.1313>, rot=<0 0 0 1>}
 base_fr_hair_l {parent=base_hair, loc=<4.5407 6.7499 -2.4319>, rot=<-0.1201 -0.9825 0.0211 0.1411>}
 dyn_fr_hair_l {parent=base_fr_hair_l, loc=<0 0 0>, rot=<0 0 0 1>}
 dyn_fr_lowerhair_l_01 {parent=dyn_fr_hair_l, loc=<8.4313 -0.4399 0.7139>, rot=<0.0286 0.2642 0.0205 -0.9638>}
 dyn_fr_lowerhair_l_02 {parent=dyn_fr_lowerhair_l_01, loc=<5.104 0.097 0.05>, rot=<0.0031 0.0219 0.2632 -0.9645>}
 base_fr_hair_r {parent=base_hair, loc=<2.2688 -7.7699 -1.1382>, rot=<-0.0242 -0.9831 0.0969 0.1533>}
 dyn_fr_hair_r {parent=base_fr_hair_r, loc=<0 0 0>, rot=<0 0 0 1>}
 dyn_fr_midhair_r {parent=dyn_fr_hair_r, loc=<6.6756 -0.6172 0.0848>, rot=<-0.1026 0.0308 0.1301 -0.9857>}
 base_fr_02_hair_r {parent=base_hair, loc=<3.6288 -7.912 -3.4577>, rot=<-0.0215 -0.9991 0.0369 0.0008>}
 dyn_fr_02_hair_r {parent=base_fr_02_hair_r, loc=<0 0 0>, rot=<0 0 0 1>}
 dyn_fr_lowerhair_r_01 {parent=dyn_fr_02_hair_r, loc=<8.588 -0.1165 -0.0261>, rot=<-0.136 0.3789 0.3659 -0.8391>}
 dyn_fr_lowerhair_r_02 {parent=dyn_fr_lowerhair_r_01, loc=<5.3673 0.9426 0.0015>, rot=<0.1223 -0.1796 -0.1936 -0.9567>}
 base_backhair_mstr {parent=base_hair, loc=<6.8203 0 -7.2829>, rot=<0 -0.9998 0 -0.021>}
 base_backhair_l {parent=base_backhair_mstr, loc=<2.8774 4.743 -0.8382>, rot=<-0.005 0.0177 0.0009 -0.9998>}
 dyn_backhair_l_01 {parent=base_backhair_l, loc=<0 0 0>, rot=<0 0 0 1>}
 dyn_backhair_l_02 {parent=dyn_backhair_l_01, loc=<4.9113 0 0.0013>, rot=<0.0139 0.0214 -0.1091 -0.9937>}
 dyn_backhair_l_03 {parent=dyn_backhair_l_02, loc=<4.533 0 0>, rot=<0.1962 0.0981 0.0197 -0.9754>}
 dyn_backhair_l_04 {parent=dyn_backhair_l_03, loc=<4.3331 0 0>, rot=<0 0.0215 0 -0.9998>}
 dyn_backhair_l_05 {parent=dyn_backhair_l_03, loc=<2.8297 3.7301 -0.1834>, rot=<0.3639 0.1386 -0.0691 -0.9185>}
 dyn_bk_hair_support_l {parent=dyn_backhair_l_01, loc=<1.1947 3.6741 -0.9576>, rot=<0.3871 0.0565 -0.1131 -0.9133>}
 base_backhair_m {parent=base_backhair_mstr, loc=<2.9034 0 -0.0008>, rot=<0 0 0 1>}
 dyn_backhair_m_01 {parent=base_backhair_m, loc=<0 0 0>, rot=<0 0 0 1>}
 dyn_backhair_m_02 {parent=dyn_backhair_m_01, loc=<4.9113 0 0.0006>, rot=<0 0.0741 0 -0.9973>}
 dyn_backhair_m_03 {parent=dyn_backhair_m_02, loc=<4.533 0 0>, rot=<0 0.0437 0 -0.999>}
 dyn_backhair_m_04 {parent=dyn_backhair_m_03, loc=<4.3331 0 0>, rot=<0 0.0215 0 -0.9998>}
 dyn_backhair_m_05 {parent=dyn_backhair_m_04, loc=<4.3842 0 0>, rot=<0 0.0775 0 -0.997>}
 base_backhair_r {parent=base_backhair_mstr, loc=<2.8774 -4.7428 -0.8382>, rot=<-0.0102 0.0541 0.0319 -0.998>}
 dyn_backhair_r_01 {parent=base_backhair_r, loc=<0 0 0>, rot=<0 0 0 -1>}
 dyn_backhair_r_02 {parent=dyn_backhair_r_01, loc=<4.9115 -0.0013 0.0028>, rot=<0 0.0741 0 -0.9973>}
 dyn_backhair_r_03 {parent=dyn_backhair_r_02, loc=<4.533 0 0>, rot=<0 0.0437 0 -0.999>}
 dyn_backhair_r_04 {parent=dyn_backhair_r_03, loc=<4.3331 0 0>, rot=<0 0.0215 0 -0.9998>}
 base_fr_hair_l_02 {parent=base_hair, loc=<4.5407 6.7499 -2.4319>, rot=<-0.1201 -0.9825 0.0211 0.1411>}
 dyn_fr_upperhair_l_01 {parent=base_fr_hair_l_02, loc=<2.8514 0.1631 -1.185>, rot=<0.0676 -0.0856 0.1158 -0.9873>}
 dyn_fr_upperhair_l_02 {parent=dyn_fr_upperhair_l_01, loc=<6.9474 0 0>, rot=<0.1742 0.3414 -0.0005 -0.9237>}
 jnt_tiny_back_strands_support {parent=neck_01, loc=<0 0 0>, rot=<0 0 0 1>}
 clavicle_l {parent=spine_05, loc=<5.4831 -0.1053 -3.2946>, rot=<0.0465 -0.7289 -0.009 -0.6829>}
 upperarm_l {parent=clavicle_l, loc=<9.8658 -1.2136 -0.4105>, rot=<0.0164 -0.3245 -0.0269 -0.9454>}
 upperarm_twist_01_l {parent=upperarm_l, loc=<7.2135 -0.1134 0>, rot=<0 0 0 1>}
 deform_upperarm_bk_base_l {parent=upperarm_twist_01_l, loc=<-6.4053 -4.3746 0.314>, rot=<0 0 0 -1>}
 deform_upperarm_bk_l {parent=deform_upperarm_bk_base_l, loc=<0 0 0>, rot=<0 0 0 1>}
 deform_upperarm_up_base_l {parent=upperarm_twist_01_l, loc=<-8.891 0.4825 3.9517>, rot=<0 0 0 -1>}
 deform_upperarm_up_l {parent=deform_upperarm_up_base_l, loc=<0 0 0>, rot=<0 0 0 1>}
 clavicle_r {parent=spine_05, loc=<5.4831 -0.1053 3.2946>, rot=<0.7289 0.0465 -0.6829 0.009>}
 upperarm_r {parent=clavicle_r, loc=<-9.8658 1.2136 0.4105>, rot=<0.0164 -0.3245 -0.0269 -0.9454>}
 upperarm_twist_01_r {parent=upperarm_r, loc=<-7.2135 0.1134 0>, rot=<0 0 0 -1>}
 deform_upperarm_bk_base_r {parent=upperarm_twist_01_r, loc=<6.4053 4.3746 -0.314>, rot=<0 0 0 -1>}
 deform_upperarm_bk_r {parent=deform_upperarm_bk_base_r, loc=<0 0 0>, rot=<0 0 0 1>}
 deform_upperarm_up_base_r {parent=upperarm_twist_01_r, loc=<8.891 -0.4825 -3.9517>, rot=<0 0 0 -1>}
 deform_upperarm_up_r {parent=deform_upperarm_up_base_r, loc=<0 0 0>, rot=<0 0 0 1>}
 mover_upperarm_r {parent=upperarm_r, loc=<0 0 0>, rot=<0 0 0 1>}
 mover_clavicle_r {parent=clavicle_r, loc=<0 0 0>, rot=<0 0 0 1>}
 attach_backpack {parent=spine_05, loc=<-0.123 -13.473 0>, rot=<-0.0093 -0.707 0.0093 -0.707>}
 attach_cape {parent=spine_05, loc=<9.017 -11.714 0>, rot=<0.0616 -0.7044 -0.0616 -0.7044>}
 base_collar_m {parent=spine_05, loc=<13.5042 -3.4438 0.0309>, rot=<0 0 -0.2326 -0.9726>}
 dyn_collar_m {parent=base_collar_m, loc=<0 0 0>, rot=<0 0 0 1>}
 base_collar_l {parent=spine_05, loc=<12.3646 -1.0585 -4.6942>, rot=<-0.469 -0.0187 -0.1658 -0.8673>}
 dyn_collar_l {parent=base_collar_l, loc=<0 0 0>, rot=<0 0 0 1>}
 base_collar_r {parent=spine_05, loc=<12.461 -1.2243 5.1851>, rot=<0.5136 -0.0035 -0.2415 -0.8233>}
 dyn_collar_r {parent=base_collar_r, loc=<0 0 0>, rot=<0 0 0 1>}
 base_collar_front_r {parent=spine_05, loc=<8.6936 5.6922 6.074>, rot=<0.4791 0.2025 -0.2077 -0.8284>}
 dyn_collar_front_r {parent=base_collar_front_r, loc=<0 0 0>, rot=<0 0 0 1>}
 base_collar_front_l {parent=spine_05, loc=<9.8769 3.7738 -5.9811>, rot=<-0.6584 -0.2493 -0.126 -0.6989>}
 dyn_collar_front_l {parent=base_collar_front_l, loc=<0 0 0>, rot=<0 0 0 1>}
]
variables [Is Skydiving {type=bool, default=False}]
