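# Animation Blueprint for the SurgeRaven female face accessory that drives its dynamic hair.
# Parent pose, with body-mesh clavicles and upper arms, feeds a Control Rig, back-hair RigidBody, then front-end or in-game front-hair RigidBody.
# Backbling or capes switch hair to bone-constrained poses, followed by additive offsets and a front-end idle constraint.
# The event graph updates these states, overwriting parameter-asset gravity with fixed skydiving or default values.
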AnimBlueprint /BRCosmetics/Characters/Player/Female/Medium/Bodies/F_MED_SurgeRaven_Pack/Meshes/Parts/F_MED_SurgeRaven_Pack_FaceAcc_AnimBP
target_skeleton=&/F_MED_SurgeRaven_Pack_FaceAcc_Skeleton
graphs [
 AnimGraph {nodes [
   "Constraint - dyn_fr_upperhair_l_02 (PPPPP)"!(
    ComponentPose
    Alpha=1
    ConstraintWeights_0=0
    ConstraintWeights_1:float=0
    ConstraintWeights_2:float=0
    ConstraintWeights_3:float=0
    ConstraintWeights_4:float=1
   ) -> (Pose) {
    alphaBoolBlend {blendOption=Linear}
    boneToModify {boneName=dyn_fr_upperhair_l_02}
    constraintSetup [
     {targetBone {boneName=spine_05}}
     {targetBone {boneName=neck_02}}
     {targetBone {boneName=neck_01}}
     {targetBone {boneName=head}}
     {targetBone {boneName=clavicle_l}}
    ]
    constraintWeights [1,1,1,1,1]
   }
   n1 = "Constraint - dyn_fr_hair_r (PPPP)"!(ComponentPose, Alpha=1, ConstraintWeights_0=0, ConstraintWeights_1:float=0, ConstraintWeights_2:float=0, ConstraintWeights_3:float=1) -> (Pose) {
    alphaBoolBlend {blendOption=Linear}
    boneToModify {boneName=dyn_fr_hair_r}
    constraintSetup [{targetBone {boneName=spine_05}}, {targetBone {boneName=neck_02}}, {targetBone {boneName=neck_01}}, {targetBone {boneName=head}}]
    constraintWeights [1,1,1,1]
   }
   "Constraint - dyn_fr_midhair_r (PPPP)"!(ComponentPose=@n1, Alpha=1, ConstraintWeights_0=0, ConstraintWeights_1:float=0, ConstraintWeights_2:float=0, ConstraintWeights_3:float=1) -> (Pose) {
    alphaBoolBlend {blendOption=Linear}
    boneToModify {boneName=dyn_fr_midhair_r}
    constraintSetup [{targetBone {boneName=spine_05}}, {targetBone {boneName=neck_02}}, {targetBone {boneName=neck_01}}, {targetBone {boneName=head}}]
    constraintWeights [1,1,1,1]
   }
   Copy Pose From Mesh() {bUseAttachedParent=true}
   "Transform (Modify) Bone - Bone: base_fr_hair_r"!(ComponentPose, bAlphaBoolEnabled:bool=True, Translation="0,0,0", Rotation="-7.5,0,0", Scale="0.9,0.9,0.9") -> (Pose) {
    alphaInputType=Bool
    rotationMode=BMM_Additive
    rotationSpace=BCS_BoneSpace
    scaleMode=BMM_Additive
    scaleSpace=BCS_BoneSpace
    translationMode=BMM_Additive
    translationSpace=BCS_BoneSpace
    alphaBoolBlend {blendInTime=0.2, blendOutTime=0.2}
    boneToModify {boneName=base_fr_hair_r}
    rotation {pitch=-7.5}
    scale=<0.9 0.9 0.9>
   }
   n2 = "Constraint - dyn_fr_hair_l (PPPPP)"!(
    ComponentPose=Local To Component(LocalPose=Use cached pose 'Dyn Pose'())
    Alpha=1
    ConstraintWeights_0=0
    ConstraintWeights_1:float=0
    ConstraintWeights_2:float=0
    ConstraintWeights_3:float=1
    ConstraintWeights_4:float=0
   ) -> (Pose) {
    alphaBoolBlend {blendOption=Linear}
    boneToModify {boneName=dyn_fr_hair_l}
    constraintSetup [
     {targetBone {boneName=spine_05}}
     {targetBone {boneName=neck_02}}
     {targetBone {boneName=neck_01}}
     {targetBone {boneName=head}}
     {targetBone {boneName=clavicle_l}}
    ]
    constraintWeights [1,1,1,1,1]
   }
   n3 = "Constraint - dyn_fr_upperhair_l_01 (PPPPP)"!(
    ComponentPose=@n2
    Alpha=1
    ConstraintWeights_0=0
    ConstraintWeights_1:float=0
    ConstraintWeights_2:float=0
    ConstraintWeights_3:float=1
    ConstraintWeights_4:float=0
   ) -> (Pose) {
    alphaBoolBlend {blendOption=Linear}
    boneToModify {boneName=dyn_fr_upperhair_l_01}
    constraintSetup [
     {targetBone {boneName=spine_05}}
     {targetBone {boneName=neck_02}}
     {targetBone {boneName=neck_01}}
     {targetBone {boneName=head}}
     {targetBone {boneName=clavicle_l}}
    ]
    constraintWeights [1,1,1,1,1]
   }
   n4 = "Constraint - dyn_fr_lowerhair_l_01 (PPPPP)"!(
    ComponentPose=@n3
    Alpha=1
    ConstraintWeights_0=0
    ConstraintWeights_1:float=0
    ConstraintWeights_2:float=0
    ConstraintWeights_3:float=0.5
    ConstraintWeights_4:float=0.5
   ) -> (Pose) {
    alphaBoolBlend {blendOption=Linear}
    boneToModify {boneName=dyn_fr_lowerhair_l_01}
    constraintSetup [
     {targetBone {boneName=spine_05}}
     {targetBone {boneName=neck_02}}
     {targetBone {boneName=neck_01}}
     {targetBone {boneName=head}}
     {targetBone {boneName=clavicle_l}}
    ]
    constraintWeights [1,1,1,1,1]
   }
   n5 = "Constraint - dyn_fr_lowerhair_l_02 (PPPPP)"!(
    ComponentPose=@n4
    Alpha=1
    ConstraintWeights_0=0.25
    ConstraintWeights_1:float=0
    ConstraintWeights_2:float=0
    ConstraintWeights_3:float=0
    ConstraintWeights_4:float=0.75
   ) -> (Pose) {
    alphaBoolBlend {blendOption=Linear}
    boneToModify {boneName=dyn_fr_lowerhair_l_02}
    constraintSetup [
     {targetBone {boneName=spine_05}}
     {targetBone {boneName=neck_02}}
     {targetBone {boneName=neck_01}}
     {targetBone {boneName=head}}
     {targetBone {boneName=clavicle_l}}
    ]
    constraintWeights [1,1,1,1,1]
   }
   n6 = "Constraint - dyn_backhair_l_01 (PPPP)"!(ComponentPose=@n5, Alpha=1, ConstraintWeights_0=0, ConstraintWeights_1:float=0, ConstraintWeights_2:float=0, ConstraintWeights_3:float=1) -> (Pose) {
    alphaBoolBlend {blendOption=Linear}
    boneToModify {boneName=dyn_backhair_l_01}
    constraintSetup [{targetBone {boneName=spine_05}}, {targetBone {boneName=neck_02}}, {targetBone {boneName=neck_01}}, {targetBone {boneName=head}}]
    constraintWeights [1,1,1,1]
   }
   n7 = "Constraint - dyn_backhair_l_02 (PPPP)"!(ComponentPose=@n6, Alpha=1, ConstraintWeights_0=0, ConstraintWeights_1:float=0, ConstraintWeights_2:float=0, ConstraintWeights_3:float=1) -> (Pose) {
    alphaBoolBlend {blendOption=Linear}
    boneToModify {boneName=dyn_backhair_l_02}
    constraintSetup [{targetBone {boneName=spine_05}}, {targetBone {boneName=neck_02}}, {targetBone {boneName=neck_01}}, {targetBone {boneName=head}}]
    constraintWeights [1,1,1,1]
   }
   n8 = "Constraint - dyn_backhair_l_03 (PPPP)"!(ComponentPose=@n7, Alpha=1, ConstraintWeights_0=1, ConstraintWeights_1:float=0, ConstraintWeights_2:float=0, ConstraintWeights_3:float=0) -> (Pose) {
    alphaBoolBlend {blendOption=Linear}
    boneToModify {boneName=dyn_backhair_l_03}
    constraintSetup [{targetBone {boneName=spine_05}}, {targetBone {boneName=neck_02}}, {targetBone {boneName=neck_01}}, {targetBone {boneName=head}}]
    constraintWeights [1,1,1,1]
   }
   n9 = "Constraint - dyn_backhair_l_04 (PPPPP)"!(
    ComponentPose=@n8
    Alpha=1
    ConstraintWeights_0=1
    ConstraintWeights_1:float=0
    ConstraintWeights_2:float=0
    ConstraintWeights_3:float=0
    ConstraintWeights_4:float=0
   ) -> (Pose) {
    alphaBoolBlend {blendOption=Linear}
    boneToModify {boneName=dyn_backhair_l_04}
    constraintSetup [
     {targetBone {boneName=spine_05}}
     {targetBone {boneName=neck_02}}
     {targetBone {boneName=neck_01}}
     {targetBone {boneName=head}}
     {targetBone {boneName=clavicle_l}}
    ]
    constraintWeights [1,1,1,1,1]
   }
   n10 = "Constraint - dyn_backhair_l_05 (PPPPP)"!(
    ComponentPose=@n9
    Alpha=1
    ConstraintWeights_0=0.5
    ConstraintWeights_1:float=0
    ConstraintWeights_2:float=0
    ConstraintWeights_3:float=0
    ConstraintWeights_4:float=0.5
   ) -> (Pose) {
    alphaBoolBlend {blendOption=Linear}
    boneToModify {boneName=dyn_backhair_l_05}
    constraintSetup [
     {targetBone {boneName=spine_05}}
     {targetBone {boneName=neck_02}}
     {targetBone {boneName=neck_01}}
     {targetBone {boneName=head}}
     {targetBone {boneName=clavicle_l}}
    ]
    constraintWeights [1,1,1,1,1]
   }
   n11 = "Constraint - dyn_backhair_m_01 (PPPP)"!(ComponentPose=@n10, Alpha=1, ConstraintWeights_0=0, ConstraintWeights_1:float=0, ConstraintWeights_2:float=0, ConstraintWeights_3:float=1) -> (Pose) {
    alphaBoolBlend {blendOption=Linear}
    boneToModify {boneName=dyn_backhair_m_01}
    constraintSetup [{targetBone {boneName=spine_05}}, {targetBone {boneName=neck_02}}, {targetBone {boneName=neck_01}}, {targetBone {boneName=head}}]
    constraintWeights [1,1,1,1]
   }
   n12 = "Constraint - dyn_backhair_m_02 (PPPP)"!(
    ComponentPose=@n11
    Alpha=1
    ConstraintWeights_0=0.5
    ConstraintWeights_1:float=0
    ConstraintWeights_2:float=0
    ConstraintWeights_3:float=0.5
   ) -> (Pose) {
    alphaBoolBlend {blendOption=Linear}
    boneToModify {boneName=dyn_backhair_m_02}
    constraintSetup [{targetBone {boneName=spine_05}}, {targetBone {boneName=neck_02}}, {targetBone {boneName=neck_01}}, {targetBone {boneName=head}}]
    constraintWeights [1,1,1,1]
   }
   n13 = "Constraint - dyn_backhair_m_03 (PPPP)"!(
    ComponentPose=@n12
    Alpha=1
    ConstraintWeights_0=0.5
    ConstraintWeights_1:float=0
    ConstraintWeights_2:float=0.5
    ConstraintWeights_3:float=0
   ) -> (Pose) {
    alphaBoolBlend {blendOption=Linear}
    boneToModify {boneName=dyn_backhair_m_03}
    constraintSetup [{targetBone {boneName=spine_05}}, {targetBone {boneName=neck_02}}, {targetBone {boneName=neck_01}}, {targetBone {boneName=head}}]
    constraintWeights [1,1,1,1]
   }
   n14 = "Constraint - dyn_backhair_m_04 (PPPP)"!(ComponentPose=@n13, Alpha=1, ConstraintWeights_0=1, ConstraintWeights_1:float=0, ConstraintWeights_2:float=0, ConstraintWeights_3:float=0) -> (Pose) {
    alphaBoolBlend {blendOption=Linear}
    boneToModify {boneName=dyn_backhair_m_04}
    constraintSetup [{targetBone {boneName=spine_05}}, {targetBone {boneName=neck_02}}, {targetBone {boneName=neck_01}}, {targetBone {boneName=head}}]
    constraintWeights [1,1,1,1]
   }
   n15 = "Constraint - dyn_backhair_m_05 (PPPP)"!(ComponentPose=@n14, Alpha=1, ConstraintWeights_0=1, ConstraintWeights_1:float=0, ConstraintWeights_2:float=0, ConstraintWeights_3:float=0) -> (Pose) {
    alphaBoolBlend {blendOption=Linear}
    boneToModify {boneName=dyn_backhair_m_05}
    constraintSetup [{targetBone {boneName=spine_05}}, {targetBone {boneName=neck_02}}, {targetBone {boneName=neck_01}}, {targetBone {boneName=head}}]
    constraintWeights [1,1,1,1]
   }
   n16 = "Constraint - dyn_backhair_r_01 (PPPP)"!(ComponentPose=@n15, Alpha=1, ConstraintWeights_0=0, ConstraintWeights_1:float=0, ConstraintWeights_2:float=0, ConstraintWeights_3:float=1) -> (Pose) {
    alphaBoolBlend {blendOption=Linear}
    boneToModify {boneName=dyn_backhair_r_01}
    constraintSetup [{targetBone {boneName=spine_05}}, {targetBone {boneName=neck_02}}, {targetBone {boneName=neck_01}}, {targetBone {boneName=head}}]
    constraintWeights [1,1,1,1]
   }
   n17 = "Constraint - dyn_backhair_r_02 (PPPP)"!(
    ComponentPose=@n16
    Alpha=1
    ConstraintWeights_0=0.5
    ConstraintWeights_1:float=0
    ConstraintWeights_2:float=0
    ConstraintWeights_3:float=0.5
   ) -> (Pose) {
    alphaBoolBlend {blendOption=Linear}
    boneToModify {boneName=dyn_backhair_r_02}
    constraintSetup [{targetBone {boneName=spine_05}}, {targetBone {boneName=neck_02}}, {targetBone {boneName=neck_01}}, {targetBone {boneName=head}}]
    constraintWeights [1,1,1,1]
   }
   n18 = "Constraint - dyn_backhair_r_03 (PPPP)"!(ComponentPose=@n17, Alpha=1, ConstraintWeights_0=0, ConstraintWeights_1:float=0, ConstraintWeights_2:float=1, ConstraintWeights_3:float=0) -> (Pose) {
    alphaBoolBlend {blendOption=Linear}
    boneToModify {boneName=dyn_backhair_r_03}
    constraintSetup [{targetBone {boneName=spine_05}}, {targetBone {boneName=neck_02}}, {targetBone {boneName=neck_01}}, {targetBone {boneName=head}}]
    constraintWeights [1,1,1,1]
   }
   n19 = "Constraint - dyn_backhair_r_04 (PPPP)"!(ComponentPose=@n18, Alpha=1, ConstraintWeights_0=1, ConstraintWeights_1:float=0, ConstraintWeights_2:float=0, ConstraintWeights_3:float=0) -> (Pose) {
    alphaBoolBlend {blendOption=Linear}
    boneToModify {boneName=dyn_backhair_r_04}
    constraintSetup [{targetBone {boneName=spine_05}}, {targetBone {boneName=neck_02}}, {targetBone {boneName=neck_01}}, {targetBone {boneName=head}}]
    constraintWeights [1,1,1,1]
   }
   n20 = "Constraint - dyn_fr_02_hair_r (PPPP)"!(ComponentPose=@n19, Alpha=1, ConstraintWeights_0=0, ConstraintWeights_1:float=0, ConstraintWeights_2:float=0, ConstraintWeights_3:float=1) -> (Pose) {
    alphaBoolBlend {blendOption=Linear}
    boneToModify {boneName=dyn_fr_02_hair_r}
    constraintSetup [{targetBone {boneName=spine_05}}, {targetBone {boneName=neck_02}}, {targetBone {boneName=neck_01}}, {targetBone {boneName=head}}]
    constraintWeights [1,1,1,1]
   }
   n21 = "Constraint - dyn_fr_lowerhair_r_01 (PPPPP)"!(
    ComponentPose=@n20
    Alpha=1
    ConstraintWeights_0=0
    ConstraintWeights_1:float=0
    ConstraintWeights_2:float=0
    ConstraintWeights_3:float=0.5
    ConstraintWeights_4:float=0.5
   ) -> (Pose) {
    alphaBoolBlend {blendOption=Linear}
    boneToModify {boneName=dyn_fr_lowerhair_r_01}
    constraintSetup [
     {targetBone {boneName=spine_05}}
     {targetBone {boneName=neck_02}}
     {targetBone {boneName=neck_01}}
     {targetBone {boneName=head}}
     {targetBone {boneName=clavicle_r}}
    ]
    constraintWeights [1,1,1,1,1]
   }
   n22 = "Constraint - dyn_fr_lowerhair_r_02 (PPPPP)"!(
    ComponentPose=@n21
    Alpha=1
    ConstraintWeights_0=0
    ConstraintWeights_1:float=0
    ConstraintWeights_2:float=0
    ConstraintWeights_3:float=0
    ConstraintWeights_4:float=1
   ) -> (Pose) {
    alphaBoolBlend {blendOption=Linear}
    boneToModify {boneName=dyn_fr_lowerhair_r_02}
    constraintSetup [
     {targetBone {boneName=spine_05}}
     {targetBone {boneName=neck_02}}
     {targetBone {boneName=neck_01}}
     {targetBone {boneName=head}}
     {targetBone {boneName=clavicle_r}}
    ]
    constraintWeights [1,1,1,1,1]
   }
   n23 = Component To Local(ComponentPose=@n22)
   n24 = Use cached pose 'Dyn Pose'()
   n25 = Get is Wearing BackBling or Cape() -> (is Wearing BackBling or Cape:bool)
   n26 = "Transform (Modify) Bone - Bone: dyn_fr_upperhair_l_02"!(
    ComponentPose=Local To Component(LocalPose=Blend Poses by bool(BlendPose_0=@n23, BlendPose_1=@n24, bActiveValue=@n25))
    bAlphaBoolEnabled:bool=True
    Translation="0,0,0"
    Rotation="-10,10,0"
    Scale="0.9,0.9,0.9"
   ) -> (Pose) {
    alphaInputType=Bool
    rotationMode=BMM_Additive
    rotationSpace=BCS_BoneSpace
    scaleMode=BMM_Additive
    scaleSpace=BCS_BoneSpace
    translationMode=BMM_Additive
    translationSpace=BCS_BoneSpace
    alphaBoolBlend {blendInTime=0.2, blendOutTime=0.2}
    boneToModify {boneName=dyn_fr_upperhair_l_02}
    rotation {pitch=-10, yaw=10}
    scale=<0.9 0.9 0.9>
   }
   n27 = "Transform (Modify) Bone - Bone: base_fr_hair_r"(ComponentPose=@n26, bAlphaBoolEnabled=True, Translation="0,0,0", Rotation="-7.5,0,0", Scale="0.9,0.9,0.9") {
    alphaInputType=Bool
    rotationMode=BMM_Additive
    rotationSpace=BCS_BoneSpace
    scaleMode=BMM_Additive
    scaleSpace=BCS_BoneSpace
    translationMode=BMM_Additive
    translationSpace=BCS_BoneSpace
    alphaBoolBlend {blendInTime=0.2, blendOutTime=0.2}
    boneToModify {boneName=base_fr_hair_r}
    rotation {pitch=-7.5}
    scale=<0.9 0.9 0.9>
   }
   n28 = "Constraint - dyn_fr_lowerhair_r_01 (PPPPP)"(
    ComponentPose=@n27
    Alpha=Get FE_Idle_Alpa() -> (FE_Idle_Alpa:double)
    ConstraintWeights_0=0
    ConstraintWeights_1=0
    ConstraintWeights_2=0
    ConstraintWeights_3=0
    ConstraintWeights_4=1
   ) {
    alphaBoolBlend {blendOption=Linear}
    boneToModify {boneName=dyn_fr_lowerhair_r_01}
    constraintSetup [
     {targetBone {boneName=spine_05}}
     {targetBone {boneName=neck_02}}
     {targetBone {boneName=neck_01}}
     {targetBone {boneName=head}}
     {targetBone {boneName=clavicle_r}}
    ]
    constraintWeights [1,1,1,1,1]
   }
   Output Pose(Result=Component To Local(ComponentPose=@n28)) {layerGroup=None}
   n29 = Get GravityOverride()
   n30 = RigidBody!(
    ComponentPose=Local To Component(LocalPose=Use cached pose 'BackHairPhysics_Cache'())
    Alpha
    OverrideWorldGravity:Vector=@n29
    ExternalForce
   ) -> (Pose) {
    bClampLinearTranslationLimitToRefPose=true
    bOverrideWorldGravity=true
    overridePhysicsAsset=&/F_MED_SurgeRaven_Pack_FaceAcc_FrontHair_Physics
    simulationSpace=BaseBoneSpace
    alphaBoolBlend {blendOption=Linear}
    baseBoneRef {boneName=neck_01}
    componentAppliedLinearAccClamp=<0 0 0>
    overrideWorldGravity {z=-980}
    simSpaceSettings {
     dampingAlpha=0
     maxAngularAcceleration=32
     maxAngularVelocity=8
     maxLinearAcceleration=1000
     maxLinearVelocity=700
     velocityScaleZ=0.7
     worldAlpha=1
     externalLinearDragV=<0.3 0.3 0.3>
    }
   }
   n31 = RigidBody(ComponentPose=Local To Component(LocalPose=Use cached pose 'BackHairPhysics_Cache'()), OverrideWorldGravity=Get GravityOverride()) {
    bClampLinearTranslationLimitToRefPose=true
    bOverrideWorldGravity=true
    overridePhysicsAsset=&/F_MED_SurgeRaven_Pack_FaceAcc_FrontHair_FE2_Physics
    simulationSpace=BaseBoneSpace
    alphaBoolBlend {blendOption=Linear}
    baseBoneRef {boneName=neck_01}
    componentAppliedLinearAccClamp=<0 0 0>
    overrideWorldGravity {z=-980}
    simSpaceSettings {
     dampingAlpha=0
     maxAngularAcceleration=32
     maxAngularVelocity=8
     maxLinearAcceleration=1000
     maxLinearVelocity=700
     velocityScaleZ=0.7
     worldAlpha=1
     externalLinearDragV=<0.3 0.3 0.3>
    }
   }
   Save cached pose 'Dyn Pose'(Pose=Blend Poses by bool(BlendPose_0=Component To Local(ComponentPose=@n31), BlendPose_1=Component To Local(ComponentPose=@n30), bActiveValue=Get bIsInFrontEnd())) {cacheName=Dyn Pose}
   n32 = Layered blend per bone(
    BasePose=Copy Pose From Mesh() {bCopyCurves=true, bUseAttachedParent=true}
    BlendPoses_0=Copy Pose From Mesh(SourceMeshComponent=Get Body_Mesh())
   ) {layerSetup [{branchFilters [
       {boneName=clavicle_l}
       {boneName=clavicle_r}
       {boneName=upperarm_l}
       {boneName=upperarm_r}
       {boneName=upperarm_twist_01_l}
       {boneName=upperarm_twist_01_r}
       {boneName=deform_upperarm_bk_base_l}
       {boneName=deform_upperarm_bk_l}
       {boneName=deform_upperarm_fr_base_l}
       {boneName=deform_upperarm_fr_l}
       {boneName=deform_upperarm_up_base_l}
       {boneName=deform_upperarm_up_l}
       {boneName=deform_upperarm_bk_base_r}
       {boneName=deform_upperarm_bk_r}
       {boneName=deform_upperarm_fr_base_r}
       {boneName=deform_upperarm_fr_r}
       {boneName=deform_upperarm_up_base_r}
       {boneName=deform_upperarm_up_r}
      ]}]}
   n33 = Control Rig!(Source=@n32, Alpha, Is Skydiving:bool=Get bIsSkyDiving()) -> (Pose) {controlRigAssetReference {blueprintRigClass=&/F_MED_SurgeRaven_Pack_FaceAcc_Test_CtrlRig.F_MED_SurgeRaven_Pack_FaceAcc_Test_CtrlRig_C}}
   n34 = RigidBody(ComponentPose=Local To Component(LocalPose=@n33), OverrideWorldGravity=@n29) {
    bClampLinearTranslationLimitToRefPose=true
    bOverrideWorldGravity=true
    overridePhysicsAsset=&/F_MED_SurgeRaven_Pack_FaceAcc_Physics
    simulationSpace=BaseBoneSpace
    alphaBoolBlend {blendOption=Linear}
    baseBoneRef {boneName=neck_01}
    componentAppliedLinearAccClamp=<0 0 0>
    overrideWorldGravity {z=-980}
    simSpaceSettings {
     dampingAlpha=0
     maxAngularAcceleration=32
     maxAngularVelocity=8
     maxLinearAcceleration=1000
     maxLinearVelocity=700
     velocityScaleZ=0.5
     worldAlpha=1
     externalLinearDragV=<0.3 0.3 0.3>
    }
   }
   Save cached pose 'BackHairPhysics_Cache'(Pose=Component To Local(ComponentPose=@n34)) {cacheName=BackHairPhysics_Cache}
   OR Boolean(A=@n25, B=Get Is Crouch Walking() -> (Is Crouch Walking:bool))
  ]}
 EventGraph {nodes [
   Event BlueprintUpdateAnimation()
   > n1 = Sequence!(execute) -> (then_0, then_1, then_2:Exec, then_3:Exec)
   > n2 = "Gravity Override State Function (Parameter Asset) (Nativized)"(
    CharacterPartAnimInstance=Self-Reference()
    GameTimeInSeconds=GetGameTimeInSeconds()
    Parameters=&/F_MED_SurgeRaven_LHair_Parameters
   )
   > Set GravityOverride(GravityOverride=@n2)
   > Set RBAN Settings Out(RBAN Settings Out=@n2.OutRigidBodyAnimNodeParameters)
   n3 = Get bIsPlayingEmote()
   > Set isFE_idle(isFE_idle:bool=AND Boolean(A=Get bIsInFrontEnd(), B=NOT Boolean(A=@n3))) -> (Output_Get:bool)
   > n4 = Branch(Condition=Get bIsSkyDiving())
   > Set GravityOverride(GravityOverride="0,0,-200")
   Set GravityOverride(execute=@n4.else, GravityOverride="0,0,-980")
   n5 = Get bIsSliding()
   Set IsCrouching not Slide(execute=@n1.then_2, IsCrouching not Slide:bool=AND Boolean(A=Get bIsCrouching(), B=NOT Boolean(A=@n5))) -> (Output_Get:bool)
   > Set Is Crouch Walking(Is Crouch Walking:bool=AND Boolean(A=Get IsCrouching not Slide() -> (IsCrouching not Slide:bool), B=Get bIsMoving2D())) -> (Output_Get:bool)
   > Set is Pickaxe swing(is Pickaxe swing:bool=Get bIsPlayingMeleeAnim()) -> (Output_Get:bool)
   n6 = Get isFE_idle() -> (isFE_idle:bool)
   > Set FE_Idle_Alpa(FE_Idle_Alpa:double=float * float(A="To Float (Boolean)"(InBool=@n6), B=0.8) {is=title K2Node_PromotableOperator}) -> (Output_Get:double)
   n7 = Cast To FortPlayerAnimInstance(execute=@n1.then_3, Object=GetAnimInstance(self=GetOwnerSkeletalMeshComponent()))
   > Set isPistolEquipped(isPistolEquipped:bool=false) -> (Output_Get:bool)
   > n8 = Switch on EFortWeaponCoreAnimation!(execute, Selection:EFortWeaponCoreAnimation Enum=Get WeaponCoreAnim(self=@n7)) -> (
    Melee:Exec
    Pistol:Exec
    Shotgun:Exec
    PaperBlueprint:Exec
    Rifle:Exec
    MeleeOneHand:Exec
    MachinePistol:Exec
    RocketLauncher:Exec
    GrenadeLauncher:Exec
    GoingCommando:Exec
    AssaultRifle:Exec
    TacticalShotgun:Exec
    SniperRifle:Exec
    TrapPlacement:Exec
    ShoulderLauncher:Exec
    AbilityDecoTool:Exec
    Crossbow:Exec
    C4:Exec
    RemoteControl:Exec
    DualWield:Exec
    AR_BullPup:Exec
    AR_ForwardGrip:Exec
    MedPackPaddles:Exec
    SMG_P90:Exec
    AR_DrumGun:Exec
    Consumable_Small:Exec
    Consumable_Large:Exec
    Balloon:Exec
    MountedTurret:Exec
    CreativeTool:Exec
    ExplosiveBow:Exec
    AshtonIndigo:Exec
    AshtonChicago:Exec
    MeleeDualWield:Exec
    Unarmed:Exec
   )
   n9 = Branch(execute=@n8.Pistol, Condition=Get bIsTargeting())
   Set isPistolEquipped(execute=[@n9.then, @n11.else], isPistolEquipped=true)
   n10 = Branch(execute=@n9.else, Condition=Get bIsRelaxedLevel1() -> (bIsRelaxedLevel1:bool))
   Set isPistolEquipped(execute=[@n10.then, @n11.then], isPistolEquipped=false)
   n11 = Branch(execute=@n10.else, Condition=Get bIsRelaxedLevel2() -> (bIsRelaxedLevel2:bool))
   n12 = CheckForBackbling()
   Set is Wearing Backbling(
    execute=@n1.then_1
    is Wearing Backbling:bool=AND Boolean!(A=@n12, B=NOT Boolean(A=@n12.bIsBackblingACape), C:bool=NOT Boolean(A=@n12.bIsBackblingLong)) -> (ReturnValue)
   ) -> (Output_Get:bool)
   > Set is Wearing Cape(is Wearing Cape:bool=OR Boolean(A=@n12.bIsBackblingACape, B=@n12.bIsBackblingLong)) -> (Output_Get:bool)
   > Set is Wearing BackBling or Cape(is Wearing BackBling or Cape:bool=OR Boolean(A=Get is Wearing Backbling() -> (is Wearing Backbling:bool), B=Get is Wearing Cape() -> (is Wearing Cape:bool))) -> (Output_Get:bool)
   n13 = Event OnCharacterCustomizationLoaded()
   > Set Body_Mesh(Body_Mesh=@n13.CustomBaseMesh)
  ]}
]
variables [
 Body_Mesh {type=SkeletalMeshComponent}
 FE_Idle_Alpa {type=number}
 Gravity Override Params {
  type=FortGravityOverrideParameterSet
  default=(Skydive_GoingUp=(GravityOverride=()),Skydive_FloatingDown=(GravityOverride=()),Skydive_DivingDown=(GravityOverride=()),Skydive_Parachuting=(GravityOverride=()),OnGround_Standing=(GravityOverride=()),OnGround_Moving=(GravityOverride=()),OnGround_Moving_HighRotationRate=(GravityOverride=()),OnGround_Crouch_Moving=(GravityOverride=()),Emote_or_Melee=(GravityOverride=()),Swimming=(GravityOverride=()),Front_End=(GravityOverride=()),DBNO=(GravityOverride=()),DBNO_Carried=(GravityOverride=()),OnGround_Falling=(GravityOverride=()),Ziplining=(GravityOverride=()),Skydive_Surfing=(GravityOverride=()),Riding_Standing=(GravityOverride=()),Riding_Moving=(GravityOverride=()),Riding_Moving_Targeting=(GravityOverride=()),Tactical_Sprint=(GravityOverride=()),Rail_Grinding=(GravityOverride=()),)
 }
 GravityOverride {type=Vector}
 Is Crouch Walking {type=boolean}
 is Pickaxe swing {type=boolean}
 is Wearing Backbling {type=boolean}
 is Wearing BackBling or Cape {type=boolean}
 is Wearing Cape {type=boolean}
 IsCrouching not Slide {type=boolean}
 isFE_idle {type=boolean}
 isPistolEquipped {type=boolean}
 RBAN Params {
  type=FortRigidBodyAnimNodeParameterSet
  default=(Skydive_GoingUp=(RigidBodyAnimNode=()),Skydive_FloatingDown=(RigidBodyAnimNode=()),Skydive_DivingDown=(RigidBodyAnimNode=()),Skydive_Parachuting=(RigidBodyAnimNode=()),OnGround_Standing=(RigidBodyAnimNode=()),OnGround_Moving=(RigidBodyAnimNode=()),OnGround_Moving_HighRotationRate=(RigidBodyAnimNode=()),OnGround_Crouch_Moving=(RigidBodyAnimNode=()),Emote_or_Melee=(RigidBodyAnimNode=()),Swimming=(RigidBodyAnimNode=()),Front_End=(RigidBodyAnimNode=()),DBNO=(RigidBodyAnimNode=()),DBNO_Carried=(RigidBodyAnimNode=()),OnGround_Falling=(RigidBodyAnimNode=()),Ziplining=(RigidBodyAnimNode=()),Skydive_Surfing=(RigidBodyAnimNode=()),Riding_Standing=(RigidBodyAnimNode=()),Riding_Moving=(RigidBodyAnimNode=()),Riding_Moving_Targeting=(RigidBodyAnimNode=()),Tactical_Sprint=(RigidBodyAnimNode=()),Rail_Grinding=(RigidBodyAnimNode=()),)
 }
 RBAN Settings Out {
  type=FortRigidBodyAnimNodeParameters
  default {simSpaceSettings {
    dampingAlpha=0
    maxAngularAcceleration=16
    maxAngularVelocity=4
    maxLinearAcceleration=800
    maxLinearVelocity=200
    velocityScaleZ=0.5
    worldAlpha=1
   }}
 }
]
