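# Control Rig for the SurgeRaven female face accessory that applies corrective offsets to the back hair bones as the head moves.
# Construction spawns head reference controls; the distance between Head_Reader and Head_Up remaps to a look-up weight.
# When not crouching, that weight drives the left, middle and right hair offsets, then targeting adds spherical-pose-reader correctives that differ for pistol.
# Crouch correctives and a second look-up pass sit behind hard-coded False branches, so never run.
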ControlRigBlueprint /BRCosmetics/Characters/Player/Female/Medium/Bodies/F_MED_SurgeRaven_Pack/Meshes/Parts/F_MED_SurgeRaven_Pack_FaceAcc_CtrlRig
graphs [
 New Function_ContainedGraph {nodes [
   RigVMFunctionEntryNode()
   > RigVMFunction_Sequence = RigVMAggregateNode()
   n1 = &RigVMFunction_ControlFlowBranch(ExecuteContext=@RigVMFunction_Sequence.B, Condition=RigVMVariableNode() {variable=Is Crouching}.Value)
   Multiply = &RigVMFunction_MathDoubleMul(A=1, B=RigVMVariableNode() {variable=SPR Look Up}.Value)
   &RigUnit_ModifyTransforms(
    ItemToModify="((Item=(Type=Bone,Name=\"None\"),Transform=(Rotation=(X=0,Y=0,Z=0,W=1),Translation=(X=0,Y=0,Z=0),Scale3D=(X=1,Y=1,Z=1))),(Item=(Type=Bone,Name=\"base_backhair_l\"),Transform=(Rotation=(X=0,Y=0.5498,Z=0,W=-0.8353),Translation=(X=0,Y=-3,Z=0),Scale3D=(X=1,Y=1,Z=1 ...(truncated)"
    WeightMinimum=0
    Weight=@Multiply.Result
    ExecutePin=@n1.False
   )
   > &RigUnit_ModifyTransforms(
    ItemToModify="((Item=(Type=Bone,Name=\"None\"),Transform=(Rotation=(X=0,Y=0,Z=0,W=1),Translation=(X=0,Y=0,Z=0),Scale3D=(X=1,Y=1,Z=1))),(Item=(Type=Bone,Name=\"base_backhair_m\"),Transform=(Rotation=(X=0,Y=-0.342,Z=0,W=0.9397),Translation=(X=0,Y=0,Z=0),Scale3D=(X=1,Y=1,Z=1)) ...(truncated)"
    WeightMinimum=0
    Weight=@Multiply.Result
   )
   > n2 = &RigUnit_ModifyTransforms(
    ItemToModify="((Item=(Type=Bone,Name=\"None\"),Transform=(Rotation=(X=0,Y=0,Z=0,W=1),Translation=(X=0,Y=0,Z=0),Scale3D=(X=1,Y=1,Z=1))),(Item=(Type=Bone,Name=\"base_backhair_r\"),Transform=(Rotation=(X=0,Y=-0.5736,Z=0,W=0.8192),Translation=(X=0,Y=3,Z=0),Scale3D=(X=1,Y=1,Z=1) ...(truncated)"
    WeightMinimum=0
    Weight=@Multiply.Result
   )
   RigVMFunction_ControlFlowBranch = &RigVMFunction_ControlFlowBranch(ExecuteContext=@n2.ExecutePin, Condition=RigVMVariableNode() {variable=Is Targeting}.Value)
   n3 = &RigVMFunction_ControlFlowBranch(ExecuteContext=@RigVMFunction_ControlFlowBranch.True, Condition=RigVMVariableNode() {variable=Is Pistol Equipped}.Value)
   n4 = &RigUnit_SphericalPoseReader(
    ActiveRegionScaleFactors=(PositiveWidth=1,NegativeWidth=2,PositiveHeight=1,NegativeHeight=1)
    ActiveRegionSize=0.1
    Debug=(bDrawDebug=false,bDraw2D=false,bDrawLocalAxes=False,DebugScale=25,DebugSegments=20,DebugThickness=0.25)
    DriverAxis=(X=1,Y=0,Z=0)
    DriverItem=(Type=Bone,Name="head")
    FalloffRegionScaleFactors=(PositiveWidth=0.1,NegativeWidth=3,PositiveHeight=1.2,NegativeHeight=1)
    FalloffSize=0.3
    FlipHeightScaling=False
    FlipWidthScaling=False
    OptionalParentItem=(Type=Bone,Name="spine_05")
    RotationOffset=(X=28.5019,Y=-25.1638,Z=-85.549)
    ExecutePin=@n3.False
   )
   > &RigUnit_ModifyTransforms(
    ItemToModify="((Item=(Type=Bone,Name=\"None\"),Transform=(Rotation=(X=0,Y=0,Z=0,W=1),Translation=(X=0,Y=0,Z=0),Scale3D=(X=1,Y=1,Z=1))),(Item=(Type=Bone,Name=\"base_backhair_l\"),Transform=(Rotation=(X=-0.1736,Y=0,Z=0,W=0.9848),Translation=(X=0,Y=0,Z=0),Scale3D=(X=1,Y=1,Z=1) ...(truncated)"
    WeightMinimum=0
    Weight=@n4.OutputParam
   )
   > &RigUnit_ModifyTransforms(
    ItemToModify="((Item=(Type=Bone,Name=\"None\"),Transform=(Rotation=(X=0,Y=0,Z=0,W=1),Translation=(X=0,Y=0,Z=0),Scale3D=(X=1,Y=1,Z=1))),(Item=(Type=Bone,Name=\"base_backhair_m\"),Transform=(Rotation=(X=0,Y=-0.342,Z=0,W=0.9397),Translation=(X=0,Y=0,Z=0),Scale3D=(X=1,Y=1,Z=1)) ...(truncated)"
    WeightMinimum=0
    Weight=@n4.OutputParam
   )
   > &RigUnit_ModifyTransforms(
    ItemToModify="((Item=(Type=Bone,Name=\"None\"),Transform=(Rotation=(X=0,Y=0,Z=0,W=1),Translation=(X=0,Y=0,Z=0),Scale3D=(X=1,Y=1,Z=1))),(Item=(Type=Bone,Name=\"base_backhair_r\"),Transform=(Rotation=(X=0,Y=0,Z=0,W=1),Translation=(X=0,Y=2,Z=3),Scale3D=(X=1,Y=1,Z=1))),(Item=(T ...(truncated)"
    WeightMinimum=0
    Weight=@n4.OutputParam
   )
   n5 = &RigUnit_SphericalPoseReader(
    ActiveRegionScaleFactors=(PositiveWidth=0.5,NegativeWidth=0.5,PositiveHeight=0.5,NegativeHeight=0.5)
    ActiveRegionSize=0.1
    Debug=(bDrawDebug=true,bDraw2D=False,bDrawLocalAxes=False,DebugScale=25,DebugSegments=20,DebugThickness=0.25)
    DriverAxis=(X=1,Y=0,Z=0)
    DriverItem=(Type=Bone,Name="head")
    FalloffRegionScaleFactors=(PositiveWidth=0.5,NegativeWidth=0.5,PositiveHeight=0.5,NegativeHeight=0.5)
    FalloffSize=0.4
    FlipHeightScaling=False
    FlipWidthScaling=False
    OptionalParentItem=(Type=Bone,Name="spine_05")
    RotationOffset=(X=-50,Y=-50,Z=0)
    ExecutePin=@n3.True
   )
   > &RigUnit_ModifyTransforms(
    ItemToModify="((Item=(Type=Bone,Name=\"None\"),Transform=(Rotation=(X=0,Y=0,Z=0,W=1),Translation=(X=0,Y=0,Z=0),Scale3D=(X=1,Y=1,Z=1))),(Item=(Type=Bone,Name=\"base_backhair_l\"),Transform=(Rotation=(X=0,Y=-0.342,Z=0,W=0.9397),Translation=(X=0,Y=0,Z=0),Scale3D=(X=1,Y=1,Z=1)) ...(truncated)"
    WeightMinimum=0
    Weight=@n5.OutputParam
   )
   > &RigUnit_ModifyTransforms(
    ItemToModify="((Item=(Type=Bone,Name=\"None\"),Transform=(Rotation=(X=0,Y=0,Z=0,W=1),Translation=(X=0,Y=0,Z=0),Scale3D=(X=1,Y=1,Z=1))),(Item=(Type=Bone,Name=\"base_backhair_m\"),Transform=(Rotation=(X=0,Y=-0.342,Z=0,W=0.9397),Translation=(X=0,Y=0,Z=0),Scale3D=(X=1,Y=1,Z=1)) ...(truncated)"
    WeightMinimum=0
    Weight=@n5.OutputParam
   )
   > &RigUnit_ModifyTransforms(
    ItemToModify="((Item=(Type=Bone,Name=\"None\"),Transform=(Rotation=(X=0,Y=0,Z=0,W=1),Translation=(X=0,Y=0,Z=0),Scale3D=(X=1,Y=1,Z=1))),(Item=(Type=Bone,Name=\"base_backhair_r\"),Transform=(Rotation=(X=0,Y=-0.342,Z=0,W=0.9397),Translation=(X=2,Y=2,Z=3.0001),Scale3D=(X=1,Y=1, ...(truncated)"
    WeightMinimum=0
    Weight=@n5.OutputParam
   )
   n6 = &RigVMFunction_ControlFlowBranch(Condition=False, ExecuteContext=@n1.True)
   n7 = &RigVMFunction_ControlFlowBranch(Condition=False, ExecuteContext=@n6.False)
   n8 = &RigUnit_SphericalPoseReader(
    ActiveRegionScaleFactors=(PositiveWidth=1.1,NegativeWidth=1,PositiveHeight=0.2,NegativeHeight=1)
    ActiveRegionSize=0.1
    Debug=(bDrawDebug=true,bDraw2D=False,bDrawLocalAxes=False,DebugScale=25,DebugSegments=20,DebugThickness=0.25)
    DriverAxis=(X=1,Y=0,Z=0)
    DriverItem=(Type=Bone,Name="head")
    FalloffRegionScaleFactors=(PositiveWidth=1,NegativeWidth=1,PositiveHeight=1,NegativeHeight=1)
    FalloffSize=0.25
    FlipHeightScaling=False
    FlipWidthScaling=False
    OptionalParentItem=(Type=Bone,Name="spine_02")
    RotationOffset=(X=-70,Y=90,Z=20)
    ExecutePin=@n7.True
   )
   > &RigUnit_ModifyTransforms(
    ItemToModify="((Item=(Type=Bone,Name=\"None\"),Transform=(Rotation=(X=0,Y=0,Z=0,W=1),Translation=(X=0,Y=0,Z=0),Scale3D=(X=1,Y=1,Z=1))),(Item=(Type=Bone,Name=\"base_backhair_l\"),Transform=(Rotation=(X=0,Y=-0.7071,Z=0,W=0.7071),Translation=(X=-6,Y=0,Z=5.0002),Scale3D=(X=1,Y= ...(truncated)"
    WeightMinimum=0
    Weight=@n8.OutputParam
   )
   > &RigUnit_ModifyTransforms(
    ItemToModify="((Item=(Type=Bone,Name=\"None\"),Transform=(Rotation=(X=0,Y=0,Z=0,W=1),Translation=(X=0,Y=0,Z=0),Scale3D=(X=1,Y=1,Z=1))),(Item=(Type=Bone,Name=\"base_backhair_m\"),Transform=(Rotation=(X=0,Y=-0.7071,Z=0,W=0.7071),Translation=(X=-6,Y=0,Z=5.0001),Scale3D=(X=1,Y= ...(truncated)"
    WeightMinimum=0
    Weight=@n8.OutputParam
   )
   > n9 = &RigUnit_ModifyTransforms(
    ItemToModify="((Item=(Type=Bone,Name=\"None\"),Transform=(Rotation=(X=0,Y=0,Z=0,W=1),Translation=(X=0,Y=0,Z=0),Scale3D=(X=1,Y=1,Z=1))),(Item=(Type=Bone,Name=\"base_backhair_r\"),Transform=(Rotation=(X=0,Y=-0.7071,Z=0,W=0.7071),Translation=(X=-6,Y=0,Z=5),Scale3D=(X=1,Y=1,Z=1 ...(truncated)"
    WeightMinimum=0
    Weight=@n8.OutputParam
   )
   n10 = &RigVMFunction_ControlFlowBranch(ExecuteContext=@n9.ExecutePin, Condition=RigVMVariableNode() {variable=Is Targeting}.Value)
   n11 = &RigUnit_SphericalPoseReader(
    ActiveRegionScaleFactors=(PositiveWidth=0.5,NegativeWidth=0.5,PositiveHeight=0.5,NegativeHeight=0.5)
    ActiveRegionSize=0.1
    Debug=(bDrawDebug=true,bDraw2D=False,bDrawLocalAxes=False,DebugScale=25,DebugSegments=20,DebugThickness=0.25)
    DriverAxis=(X=1,Y=0,Z=0)
    DriverItem=(Type=Bone,Name="head")
    FalloffRegionScaleFactors=(PositiveWidth=1,NegativeWidth=1,PositiveHeight=1,NegativeHeight=1)
    FalloffSize=0.4
    FlipHeightScaling=False
    FlipWidthScaling=False
    OptionalParentItem=(Type=Bone,Name="spine_05")
    RotationOffset=(X=-50,Y=0,Z=0)
    ExecutePin=@n10.True
   )
   > &RigUnit_ModifyTransforms(
    ItemToModify="((Item=(Type=Bone,Name=\"None\"),Transform=(Rotation=(X=0,Y=0,Z=0,W=1),Translation=(X=0,Y=0,Z=0),Scale3D=(X=1,Y=1,Z=1))),(Item=(Type=Bone,Name=\"base_backhair_l\"),Transform=(Rotation=(X=0,Y=0,Z=0,W=1),Translation=(X=0,Y=0,Z=5),Scale3D=(X=1,Y=1,Z=1))),(Item=(T ...(truncated)"
    WeightMinimum=0
    Weight=@n11.OutputParam
   )
   > &RigUnit_ModifyTransforms(
    ItemToModify=((Item=(Type=Bone,Name="None"),Transform=(Rotation=(X=0,Y=0,Z=0,W=1),Translation=(X=0,Y=0,Z=0),Scale3D=(X=1,Y=1,Z=1))),(Item=(Type=Bone,Name="base_backhair_m"),Transform=(Rotation=(X=0,Y=0,Z=0,W=1),Translation=(X=-3,Y=0,Z=5),Scale3D=(X=1,Y=1,Z=1))))
    WeightMinimum=0
    Weight=@n11.OutputParam
   )
   > &RigUnit_ModifyTransforms(
    ItemToModify="((Item=(Type=Bone,Name=\"None\"),Transform=(Rotation=(X=0,Y=0,Z=0,W=1),Translation=(X=0,Y=0,Z=0),Scale3D=(X=1,Y=1,Z=1))),(Item=(Type=Bone,Name=\"base_backhair_r\"),Transform=(Rotation=(X=0,Y=0,Z=0,W=1),Translation=(X=-3,Y=0,Z=5),Scale3D=(X=1,Y=1,Z=1))),(Item=( ...(truncated)"
    WeightMinimum=0
    Weight=@n11.OutputParam
   )
   n12 = &RigUnit_SphericalPoseReader(
    ActiveRegionScaleFactors=(PositiveWidth=0.5,NegativeWidth=0.5,PositiveHeight=0.5,NegativeHeight=0.5)
    ActiveRegionSize=0.1
    Debug=(bDrawDebug=false,bDraw2D=False,bDrawLocalAxes=False,DebugScale=25,DebugSegments=20,DebugThickness=0.25)
    DriverAxis=(X=1,Y=0,Z=0)
    DriverItem=(Type=Bone,Name="head")
    FalloffRegionScaleFactors=(PositiveWidth=0.3,NegativeWidth=0.3,PositiveHeight=0.3,NegativeHeight=0.3)
    FalloffSize=0.4
    FlipHeightScaling=False
    FlipWidthScaling=False
    OptionalParentItem=(Type=Bone,Name="spine_05")
    RotationOffset=(X=-70,Y=0,Z=20)
    ExecutePin=@n6.True
   )
   > &RigUnit_ModifyTransforms(
    ItemToModify="((Item=(Type=Bone,Name=\"None\"),Transform=(Rotation=(X=0,Y=0,Z=0,W=1),Translation=(X=0,Y=0,Z=0),Scale3D=(X=1,Y=1,Z=1))),(Item=(Type=Bone,Name=\"base_backhair_l\"),Transform=(Rotation=(X=0,Y=-0.7071,Z=0,W=0.7071),Translation=(X=0,Y=0,Z=3),Scale3D=(X=1,Y=1,Z=1) ...(truncated)"
    WeightMinimum=0
    Weight=@n12.OutputParam
   )
   > &RigUnit_ModifyTransforms(
    ItemToModify="((Item=(Type=Bone,Name=\"None\"),Transform=(Rotation=(X=0,Y=0,Z=0,W=1),Translation=(X=0,Y=0,Z=0),Scale3D=(X=1,Y=1,Z=1))),(Item=(Type=Bone,Name=\"base_backhair_m\"),Transform=(Rotation=(X=0,Y=-0.7071,Z=0,W=0.7071),Translation=(X=0,Y=0,Z=3),Scale3D=(X=1,Y=1,Z=1) ...(truncated)"
    WeightMinimum=0
    Weight=@n12.OutputParam
   )
   > n13 = &RigUnit_ModifyTransforms(
    ItemToModify="((Item=(Type=Bone,Name=\"None\"),Transform=(Rotation=(X=0,Y=0,Z=0,W=1),Translation=(X=0,Y=0,Z=0),Scale3D=(X=1,Y=1,Z=1))),(Item=(Type=Bone,Name=\"base_backhair_r\"),Transform=(Rotation=(X=0,Y=-0.7071,Z=0,W=0.7071),Translation=(X=0,Y=0,Z=3),Scale3D=(X=1,Y=1,Z=1) ...(truncated)"
    WeightMinimum=0
    Weight=@n12.OutputParam
   )
   n14 = &RigVMFunction_ControlFlowBranch(ExecuteContext=@n13.ExecutePin, Condition=RigVMVariableNode() {variable=Is Targeting}.Value)
   n15 = &RigVMFunction_ControlFlowBranch(ExecuteContext=@n14.True, Condition=RigVMVariableNode() {variable=Is Pistol Equipped}.Value)
   n16 = &RigUnit_SphericalPoseReader(
    ActiveRegionScaleFactors=(PositiveWidth=1,NegativeWidth=2,PositiveHeight=1,NegativeHeight=1)
    ActiveRegionSize=0.1
    Debug=(bDrawDebug=false,bDraw2D=false,bDrawLocalAxes=False,DebugScale=25,DebugSegments=20,DebugThickness=0.25)
    DriverAxis=(X=1,Y=0,Z=0)
    DriverItem=(Type=Bone,Name="head")
    FalloffRegionScaleFactors=(PositiveWidth=0.1,NegativeWidth=3,PositiveHeight=1.2,NegativeHeight=1)
    FalloffSize=0.3
    FlipHeightScaling=False
    FlipWidthScaling=False
    OptionalParentItem=(Type=Bone,Name="spine_05")
    RotationOffset=(X=28.5019,Y=-25.1638,Z=-85.549)
    ExecutePin=@n15.False
   )
   > &RigUnit_ModifyTransforms(
    ItemToModify="((Item=(Type=Bone,Name=\"None\"),Transform=(Rotation=(X=0,Y=0,Z=0,W=1),Translation=(X=0,Y=0,Z=0),Scale3D=(X=1,Y=1,Z=1))),(Item=(Type=Bone,Name=\"base_backhair_l\"),Transform=(Rotation=(X=-0.1736,Y=0,Z=0,W=0.9848),Translation=(X=0,Y=0,Z=0),Scale3D=(X=1,Y=1,Z=1) ...(truncated)"
    WeightMinimum=0
    Weight=@n16.OutputParam
   )
   > &RigUnit_ModifyTransforms(
    ItemToModify="((Item=(Type=Bone,Name=\"None\"),Transform=(Rotation=(X=0,Y=0,Z=0,W=1),Translation=(X=0,Y=0,Z=0),Scale3D=(X=1,Y=1,Z=1))),(Item=(Type=Bone,Name=\"base_backhair_m\"),Transform=(Rotation=(X=0,Y=-0.342,Z=0,W=0.9397),Translation=(X=0,Y=0,Z=0),Scale3D=(X=1,Y=1,Z=1)) ...(truncated)"
    WeightMinimum=0
    Weight=@n16.OutputParam
   )
   > &RigUnit_ModifyTransforms(
    ItemToModify="((Item=(Type=Bone,Name=\"None\"),Transform=(Rotation=(X=0,Y=0,Z=0,W=1),Translation=(X=0,Y=0,Z=0),Scale3D=(X=1,Y=1,Z=1))),(Item=(Type=Bone,Name=\"base_backhair_r\"),Transform=(Rotation=(X=0,Y=-0.342,Z=0,W=0.9397),Translation=(X=0,Y=2,Z=3),Scale3D=(X=1,Y=1,Z=1)) ...(truncated)"
    WeightMinimum=0
    Weight=@n16.OutputParam
   )
   n17 = &RigVMFunction_ControlFlowBranch(Condition=False, ExecuteContext=@RigVMFunction_Sequence.C)
   n18 = RigVMVariableNode() {variable=SPR Look Up}
   &RigUnit_ModifyTransforms(
    ItemToModify="((Item=(Type=Bone,Name=\"None\"),Transform=(Rotation=(X=0,Y=0,Z=0,W=1),Translation=(X=0,Y=0,Z=0),Scale3D=(X=1,Y=1,Z=1))),(Item=(Type=Bone,Name=\"base_backhair_l\"),Transform=(Rotation=(X=0,Y=0,Z=0,W=1),Translation=(X=0,Y=0,Z=0),Scale3D=(X=1,Y=1,Z=1))),(Item=(T ...(truncated)"
    WeightMinimum=0
    ExecutePin=@n17.True
    Weight=@n18.Value
   )
   > &RigUnit_ModifyTransforms(
    ItemToModify="((Item=(Type=Bone,Name=\"None\"),Transform=(Rotation=(X=0,Y=0,Z=0,W=1),Translation=(X=0,Y=0,Z=0),Scale3D=(X=1,Y=1,Z=1))),(Item=(Type=Bone,Name=\"base_backhair_m\"),Transform=(Rotation=(X=0,Y=-0.2588,Z=0,W=0.9659),Translation=(X=0,Y=0,Z=0),Scale3D=(X=1,Y=1,Z=1) ...(truncated)"
    WeightMinimum=0
    Weight=@n18.Value
   )
   > &RigUnit_ModifyTransforms(
    ItemToModify="((Item=(Type=Bone,Name=\"None\"),Transform=(Rotation=(X=0,Y=0,Z=0,W=1),Translation=(X=0,Y=0,Z=0),Scale3D=(X=1,Y=1,Z=1))),(Item=(Type=Bone,Name=\"base_backhair_r\"),Transform=(Rotation=(X=0,Y=0,Z=0,W=1),Translation=(X=0,Y=0,Z=0),Scale3D=(X=1,Y=1,Z=1))),(Item=(T ...(truncated)"
    WeightMinimum=0
    Weight=@n18.Value
   )
   RigVMFunctionReturnNode(ExecuteContext=@RigVMFunction_Sequence.D)
   RigVMVariableNode(
    Value=&RigUnit_GetTransform(Item=(Type=Control,Name="Head_Reader_ctrl"), bInitial=False).Transform.Translation
    ExecuteContext=@RigVMFunction_Sequence.A
   ) {variable=Set_Reader_Translation}
   RigVMFunction_MathVectorDistance = &RigVMFunction_MathVectorDistance(
    B=&RigUnit_GetTransform(Item=(Type=Control,Name="Head_Up_ctrl"), bInitial=False).Transform.Translation
    A=RigVMVariableNode() {variable=Set_Reader_Translation}.Value
   )
   Remap = &RigVMFunction_MathDoubleRemap(SourceMaximum=2, SourceMinimum=21, TargetMaximum=1, TargetMinimum=0, Value=@RigVMFunction_MathVectorDistance.Result)
   > RigVMVariableNode(Value=@Remap.Result) {variable=SPR Look Up}
   n19 = &RigVMFunction_AlphaInterp(
    InterpSpeedDecreasing=10
    InterpSpeedIncreasing=10
    Scale=1
    Value=0
    bClampResult=False
    bInterpResult=RigVMVariableNode() {variable=Is Crouching}.Value
   )
   &RigVMFunction_AnimEvalRichCurve(
    Curve="(EditorCurveData=(Keys=((InterpMode=RCIM_Linear,TangentMode=RCTM_Auto,TangentWeightMode=RCTWM_WeightedNone,Time=0,Value=0,ArriveTangent=0,ArriveTangentWeight=0,LeaveTangent=0,LeaveTangentWeight=0),(InterpMode=RCIM_Linear,TangentMode=RCTM_Auto,TangentWeight ...(truncated)"
    SourceMinimum=0
    TargetMinimum=0
    Value=@n19.Result
   )
   RigVMFunction_MathBoolAnd = &RigVMFunction_MathBoolAnd(A=&RigVMFunction_MathBoolNot(Value=False).Result, B=RigVMVariableNode() {variable=Is Crouching}.Value)
  ]}
 RigVMFunctionLibrary {nodes [Hair Correctives = RigVMCollapseNode(ExecuteContext=())]}
 RigVMFunction_Sequence_1_ContainedGraph {nodes [
   n1 = RigVMFunctionEntryNode()
   RigVMFunction_Sequence = &RigVMFunction_Sequence(ExecuteContext=@n1.ExecuteContext)
   n2 = &RigVMFunction_Sequence(ExecuteContext=@RigVMFunction_Sequence.B)
   n3 = &RigVMFunction_Sequence(ExecuteContext=@n2.B)
   n4 = &RigVMFunction_Sequence(ExecuteContext=@n3.B)
   RigVMFunctionReturnNode(A=@RigVMFunction_Sequence.A, B=@n2.A, C=@n3.A, D=@n4.A, E=@n4.B)
  ]}
 RigVMModel {nodes [
   n1 = &RigUnit_PrepareForExecution()
   Control Stack at Item = RigVMFunctionReferenceNode(
    Add Bottom Null=False
    Add Null=true
    Add Secondary Control=False
    Control Suffix=_ctrl
    Item=(Name="head",Type=Bone)
    Name=Head_Reader
    Null Suffix=_null
    Parent=(Name="head",Type=Bone)
    Secondary Settings="(Shape=(bVisible=True,Name=\"Default\",Color=(R=1,G=0,B=0,A=1),Transform=(Rotation=(X=0,Y=0,Z=0,W=1),Translation=(X=0,Y=0,Z=0),Scale3D=(X=1,Y=1,Z=1))),Proxy=(bIsProxy=False,ShapeVisibility=BasedOnSelection),DisplayName=\"None\",InitialSpace=LocalSpace,bUsePref ...(truncated)"
    Settings="(Shape=(bVisible=True,Name=\"Default\",Color=(R=1,G=0,B=0,A=1),Transform=(Rotation=(X=0,Y=0,Z=0,W=1),Translation=(X=0,Y=0,Z=0),Scale3D=(X=0.5,Y=0.5,Z=0.5))),Proxy=(bIsProxy=False,ShapeVisibility=BasedOnSelection),DisplayName=\"None\",InitialSpace=LocalSpace,bU ...(truncated)"
    ExecuteContext=@n1.ExecutePin
   )
   > SetControlOffset = /Script/ControlRig.RigUnit_SetControlOffset(
    Offset=(Rotation=(X=0,Y=0,Z=0,W=1),Translation=(X=25,Y=0,Z=0),Scale3D=(X=1,Y=1,Z=1))
    Space=LocalSpace
    Control=@"Control Stack at Item".Control.Name
   ) {is=struct RigVMUnitNode, template="SetControlOffset::Execute(in Control,io Offset,in Space)"}
   > Control Stack at Item_1 = RigVMFunctionReferenceNode(
    Add Bottom Null=False
    Add Null=true
    Add Secondary Control=False
    Control Suffix=_ctrl
    Item=(Name="head",Type=Bone)
    Name=Head_Up
    Null Suffix=_null
    Parent=(Name="spine_05",Type=Bone)
    Secondary Settings="(Shape=(bVisible=True,Name=\"Default\",Color=(R=1,G=0,B=0,A=1),Transform=(Rotation=(X=0,Y=0,Z=0,W=1),Translation=(X=0,Y=0,Z=0),Scale3D=(X=1,Y=1,Z=1))),Proxy=(bIsProxy=False,ShapeVisibility=BasedOnSelection),DisplayName=\"None\",InitialSpace=LocalSpace,bUsePref ...(truncated)"
    Settings="(Shape=(bVisible=True,Name=\"Default\",Color=(R=0,G=0,B=1,A=1),Transform=(Rotation=(X=0,Y=0,Z=0,W=1),Translation=(X=0,Y=0,Z=0),Scale3D=(X=0.5,Y=0.5,Z=0.5))),Proxy=(bIsProxy=False,ShapeVisibility=BasedOnSelection),DisplayName=\"None\",InitialSpace=LocalSpace,bU ...(truncated)"
   )
   > &RigUnit_SetControlOffset(
    Offset=(Rotation=(X=0,Y=0,Z=0,W=1),Translation=(X=16,Y=-19.2,Z=0),Scale3D=(X=1,Y=1,Z=1))
    Space=LocalSpace
    Control=@"Control Stack at Item_1".Control.Name
   ) {template="SetControlOffset::Execute(in Control,io Offset,in Space)"}
   > Control Stack at Item_2 = RigVMFunctionReferenceNode(
    Add Bottom Null=False
    Add Null=true
    Add Secondary Control=False
    Control Suffix=_ctrl
    Item=(Name="head",Type=Bone)
    Name=Head_aim
    Null Suffix=_null
    Parent=(Name="spine_04",Type=Bone)
    Secondary Settings="(Shape=(bVisible=True,Name=\"Default\",Color=(R=1,G=0,B=0,A=1),Transform=(Rotation=(X=0,Y=0,Z=0,W=1),Translation=(X=0,Y=0,Z=0),Scale3D=(X=1,Y=1,Z=1))),Proxy=(bIsProxy=False,ShapeVisibility=BasedOnSelection),DisplayName=\"None\",InitialSpace=LocalSpace,bUsePref ...(truncated)"
    Settings="(Shape=(bVisible=True,Name=\"Default\",Color=(R=1,G=0,B=1,A=1),Transform=(Rotation=(X=0,Y=0,Z=0,W=1),Translation=(X=0,Y=0,Z=0),Scale3D=(X=0.5,Y=0.5,Z=0.5))),Proxy=(bIsProxy=False,ShapeVisibility=BasedOnSelection),DisplayName=\"None\",InitialSpace=LocalSpace,bU ...(truncated)"
   )
   > &RigUnit_SetControlOffset(
    Offset=(Rotation=(X=0,Y=0,Z=0,W=1),Translation=(X=16,Y=-10,Z=20),Scale3D=(X=1,Y=1,Z=1))
    Space=LocalSpace
    Control=@"Control Stack at Item_2".Control.Name
   ) {template="SetControlOffset::Execute(in Control,io Offset,in Space)"}
   > Control Stack at Item_3 = RigVMFunctionReferenceNode(
    Add Bottom Null=False
    Add Null=true
    Add Secondary Control=False
    Control Suffix=_ctrl
    Item=(Name="head",Type=Bone)
    Name=Head_Crouch_Aim
    Null Suffix=_null
    Parent=(Name="spine_04",Type=Bone)
    Secondary Settings="(Shape=(bVisible=True,Name=\"Default\",Color=(R=1,G=0,B=0,A=1),Transform=(Rotation=(X=0,Y=0,Z=0,W=1),Translation=(X=0,Y=0,Z=0),Scale3D=(X=1,Y=1,Z=1))),Proxy=(bIsProxy=False,ShapeVisibility=BasedOnSelection),DisplayName=\"None\",InitialSpace=LocalSpace,bUsePref ...(truncated)"
    Settings="(Shape=(bVisible=True,Name=\"Default\",Color=(R=0,G=0.9,B=0.9,A=1),Transform=(Rotation=(X=0,Y=0,Z=0,W=1),Translation=(X=0,Y=0,Z=0),Scale3D=(X=0.5,Y=0.5,Z=0.5))),Proxy=(bIsProxy=False,ShapeVisibility=BasedOnSelection),DisplayName=\"None\",InitialSpace=LocalSpac ...(truncated)"
   )
   > &RigUnit_SetControlOffset(Control=Head_Crouch_Aim_ctrl, Offset=(Rotation=(X=0,Y=0,Z=0,W=1),Translation=(X=-3,Y=-32,Z=20),Scale3D=(X=1,Y=1,Z=1)), Space=LocalSpace) {template="SetControlOffset::Execute(in Control,io Offset,in Space)"}
   RigUnit_BeginExecution = &RigUnit_BeginExecution()
   Hair Correctives_1 = RigVMFunctionReferenceNode(ExecuteContext=@RigUnit_BeginExecution.ExecutePin)
   addFKCtrl = RigVMFunctionReferenceNode(
    Custom_FKCtrl=()
    ExecuteContext=()
    Options_FKCtrl="(pelvis_1_#1=False,spine_3_#2=false,clavicle_5_#3=true,arm_7_#4=true,leg_9_#5=False,head_11_46DC15EF465CD2 ...(truncated)"
    Shape_Settings="(pelvis_2_#6=(bVisible=True,Name=\"Hexagon_Thin\",Color=(R=0.0117,G=0.8281,B=1,A=1),Transform=(Rotation=(X=0,Y=-0.7071,Z=0,W=0.7071),Translation=(X=2,Y=0,Z=0),Scale3D=(X=5,Y=5,Z=7))),spine_4_#7=(bVi ...(truncated)"
   )
   setFKCtrl = RigVMFunctionReferenceNode(ExecuteContext=(), setFK_Ctrl=false)
  ]}
]
hierarchy [
 root {loc=<0 0 0>, rot=<0 0 0 1>}
 pelvis {parent=root, loc=<0 -0.5301 98.0571>, rot=<0.0127 0.707 -0.0127 -0.707>}
 spine_01 {parent=pelvis, loc=<6.1208 -0.9328 0>, rot=<0 0 0.034 -0.9994>}
 spine_02 {parent=spine_01, loc=<6.5997 -0.2095 0>, rot=<0 0 0.0238 -0.9997>}
 spine_03 {parent=spine_02, loc=<6.4914 -0.1237 0>, rot=<0 0 0.0149 -0.9999>}
 spine_04 {parent=spine_03, loc=<6.9598 0.0794 0>, rot=<0 0 -0.0132 -0.9999>}
 spine_05 {parent=spine_04, loc=<6.9987 0 0>, rot=<0 0 -0.0015 -1>}
 neck_01 {parent=spine_05, loc=<11.3138 0.1529 0>, rot=<0 0 -0.2003 -0.9797>}
 neck_02 {parent=neck_01, loc=<4.5627 0 0>, rot=<0 0 0.0504 -0.9987>}
 head {parent=neck_02, loc=<5.3914 0 0>, rot=<0 0 0.115 -0.9934>}
 base_hair {parent=head, loc=<0 0 0>, rot=<0.7071 -0.0009 0.0009 -0.7071>}
 base_hairtop {parent=base_hair, loc=<12.0279 0 0.1313>, rot=<0 0 0 1>}
 base_fr_hair_l {parent=base_hair, loc=<4.5407 6.7499 -2.4319>, rot=<-0.1201 -0.9825 0.0211 0.1411>}
 dyn_fr_hair_l {parent=base_fr_hair_l, loc=<0 0 0>, rot=<0 0 0 1>}
 dyn_fr_lowerhair_l_01 {parent=dyn_fr_hair_l, loc=<8.4313 -0.4399 0.7139>, rot=<0.0286 0.2642 0.0205 -0.9638>}
 dyn_fr_lowerhair_l_02 {parent=dyn_fr_lowerhair_l_01, loc=<5.104 0.097 0.05>, rot=<0.0031 0.0219 0.2632 -0.9645>}
 dyn_fr_upperhair_l_01 {parent=base_fr_hair_l, loc=<2.8514 0.1631 -1.185>, rot=<0.0676 -0.0856 0.1158 -0.9873>}
 dyn_fr_upperhair_l_02 {parent=dyn_fr_upperhair_l_01, loc=<6.9474 0 0>, rot=<0.1742 0.3414 -0.0005 -0.9237>}
 base_fr_hair_r {parent=base_hair, loc=<2.2688 -7.7699 -1.1382>, rot=<-0.0242 -0.9831 0.0969 0.1533>}
 dyn_fr_hair_r {parent=base_fr_hair_r, loc=<0 0 0>, rot=<0 0 0 1>}
 dyn_fr_midhair_r {parent=dyn_fr_hair_r, loc=<6.6756 -0.6172 0.0848>, rot=<-0.1026 0.0308 0.1301 -0.9857>}
 base_backhair_l {parent=base_hair, loc=<2.8774 4.743 -0.8382>, rot=<-0.005 0.0177 0.0009 -0.9998>}
 dyn_backhair_l_01 {parent=base_backhair_l, loc=<0 0 0>, rot=<0 0 0 1>}
 dyn_backhair_l_02 {parent=dyn_backhair_l_01, loc=<4.9113 0 0.0013>, rot=<0.0139 0.0214 -0.1091 -0.9937>}
 dyn_backhair_l_03 {parent=dyn_backhair_l_02, loc=<4.533 0 0>, rot=<0.1962 0.0981 0.0197 -0.9754>}
 dyn_backhair_l_04 {parent=dyn_backhair_l_03, loc=<4.3331 0 0>, rot=<0 0.0215 0 -0.9998>}
 dyn_backhair_l_05 {parent=dyn_backhair_l_03, loc=<2.8297 3.7301 -0.1834>, rot=<0.3639 0.1386 -0.0691 -0.9185>}
 base_backhair_m {parent=base_hair, loc=<2.9034 0 -0.0008>, rot=<0 0 0 1>}
 dyn_backhair_m_01 {parent=base_backhair_m, loc=<0 0 0>, rot=<0 0 0 1>}
 dyn_backhair_m_02 {parent=dyn_backhair_m_01, loc=<4.9113 0 0.0006>, rot=<0 0.0741 0 -0.9973>}
 dyn_backhair_m_03 {parent=dyn_backhair_m_02, loc=<4.533 0 0>, rot=<0 0.0437 0 -0.999>}
 dyn_backhair_m_04 {parent=dyn_backhair_m_03, loc=<4.3331 0 0>, rot=<0 0.0215 0 -0.9998>}
 dyn_backhair_m_05 {parent=dyn_backhair_m_04, loc=<4.3842 0 0>, rot=<0 0.0775 0 -0.997>}
 base_backhair_r {parent=base_hair, loc=<2.8774 -4.7428 -0.8382>, rot=<-0.0102 0.0541 0.0319 -0.998>}
 dyn_backhair_r_01 {parent=base_backhair_r, loc=<0 0 0>, rot=<0 0 0 -1>}
 dyn_backhair_r_02 {parent=dyn_backhair_r_01, loc=<4.9115 -0.0013 0.0028>, rot=<0 0.0741 0 -0.9973>}
 dyn_backhair_r_03 {parent=dyn_backhair_r_02, loc=<4.533 0 0>, rot=<0 0.0437 0 -0.999>}
 dyn_backhair_r_04 {parent=dyn_backhair_r_03, loc=<4.3331 0 0>, rot=<0 0.0215 0 -0.9998>}
 base_fr_02_hair_r {parent=base_hair, loc=<3.6288 -7.912 -3.4577>, rot=<-0.0215 -0.9991 0.0369 0.0008>}
 dyn_fr_02_hair_r {parent=base_fr_02_hair_r, loc=<0 0 0>, rot=<0 0 0 1>}
 dyn_fr_lowerhair_r_01 {parent=dyn_fr_02_hair_r, loc=<8.588 -0.1165 -0.0261>, rot=<-0.136 0.3789 0.3659 -0.8391>}
 dyn_fr_lowerhair_r_02 {parent=dyn_fr_lowerhair_r_01, loc=<5.3673 0.9426 0.0015>, rot=<0.1223 -0.1796 -0.1936 -0.9567>}
 clavicle_l {parent=spine_05, loc=<5.4831 -0.1053 -3.2946>, rot=<0.0465 -0.7289 -0.009 -0.6829>}
 upperarm_l {parent=clavicle_l, loc=<9.8658 -1.2136 -0.4105>, rot=<0.0164 -0.3245 -0.0269 -0.9454>}
 upperarm_twist_01_l {parent=upperarm_l, loc=<7.2135 -0.1134 0>, rot=<0 0 0 1>}
 deform_upperarm_bk_base_l {parent=upperarm_twist_01_l, loc=<-6.4053 -4.3746 0.314>, rot=<0 0 0 -1>}
 deform_upperarm_bk_l {parent=deform_upperarm_bk_base_l, loc=<0 0 0>, rot=<0 0 0 1>}
 deform_upperarm_up_base_l {parent=upperarm_twist_01_l, loc=<-8.891 0.4825 3.9517>, rot=<0 0 0 -1>}
 deform_upperarm_up_l {parent=deform_upperarm_up_base_l, loc=<0 0 0>, rot=<0 0 0 1>}
 clavicle_r {parent=spine_05, loc=<5.4831 -0.1053 3.2946>, rot=<0.7289 0.0465 -0.6829 0.009>}
 upperarm_r {parent=clavicle_r, loc=<-9.8658 1.2136 0.4105>, rot=<0.0164 -0.3245 -0.0269 -0.9454>}
 upperarm_twist_01_r {parent=upperarm_r, loc=<-7.2135 0.1134 0>, rot=<0 0 0 -1>}
 deform_upperarm_bk_base_r {parent=upperarm_twist_01_r, loc=<6.4053 4.3746 -0.314>, rot=<0 0 0 -1>}
 deform_upperarm_bk_r {parent=deform_upperarm_bk_base_r, loc=<0 0 0>, rot=<0 0 0 1>}
 deform_upperarm_up_base_r {parent=upperarm_twist_01_r, loc=<8.891 -0.4825 -3.9517>, rot=<0 0 0 -1>}
 deform_upperarm_up_r {parent=deform_upperarm_up_base_r, loc=<0 0 0>, rot=<0 0 0 1>}
 attach_backpack {parent=spine_05, loc=<-0.123 -13.473 0>, rot=<-0.0093 -0.707 0.0093 -0.707>}
 attach_cape {parent=spine_05, loc=<9.017 -11.714 0>, rot=<0.0616 -0.7044 -0.0616 -0.7044>}
 base_collar_m {parent=spine_05, loc=<13.5042 -3.4438 0.0309>, rot=<0 0 -0.2326 -0.9726>}
 dyn_collar_m {parent=base_collar_m, loc=<0 0 0>, rot=<0 0 0 1>}
 base_collar_l {parent=spine_05, loc=<12.3646 -1.0585 -4.6942>, rot=<-0.469 -0.0187 -0.1658 -0.8673>}
 dyn_collar_l {parent=base_collar_l, loc=<0 0 0>, rot=<0 0 0 1>}
 base_collar_r {parent=spine_05, loc=<12.461 -1.2243 5.1851>, rot=<0.5136 -0.0035 -0.2415 -0.8233>}
 dyn_collar_r {parent=base_collar_r, loc=<0 0 0>, rot=<0 0 0 1>}
 base_collar_front_r {parent=spine_05, loc=<8.6936 5.6922 6.074>, rot=<0.4791 0.2025 -0.2077 -0.8284>}
 dyn_collar_front_r {parent=base_collar_front_r, loc=<0 0 0>, rot=<0 0 0 1>}
 base_collar_front_l {parent=spine_05, loc=<9.8769 3.7738 -5.9811>, rot=<-0.6584 -0.2493 -0.126 -0.6989>}
 dyn_collar_front_l {parent=base_collar_front_l, loc=<0 0 0>, rot=<0 0 0 1>}
]
variables [
 Is Crouching {type=bool, default=False}
 Is Crouching Moving {type=bool, default=False}
 Is Crouching Not Moving {type=bool, default=False}
 Is Crouching not Sliding {type=bool, default=False}
 Is Pistol Equipped {type=bool, default=False}
 Is Skydiving {type=bool, default=False}
 Is Targeting {type=bool, default=False}
 SPR Look Up {type=double, default=0}
 Set_Reader_Translation {type=FVector, default=(X=0,Y=0,Z=0)}
]
